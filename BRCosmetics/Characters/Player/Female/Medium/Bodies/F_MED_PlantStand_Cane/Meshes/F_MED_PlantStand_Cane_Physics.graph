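# Physics asset for the F_MED_PlantStand_Cane female player body, setting up jointed capsule bodies for its dynamic skirt and sash bones.
# Skirt and sash chains hang from master bones on the pelvis, with free rotation at the masters and default-type limited swing and twist down each chain.
# Pelvis and leg capsules supply body volume, and a huge root capsule below the origin approximates a ground surface.
# An extra free-rotation constraint ties the left front skirt tip to the left side tip.
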PhysicsAsset /BRCosmetics/Characters/Player/Female/Medium/Bodies/F_MED_PlantStand_Cane/Meshes/F_MED_PlantStand_Cane_Physics
bodies {
 calf_l [{shapeType=Capsule, length=26.2042, radius=12.5842, shapeName=calf_l_capsule, center=<-23.0325 -1.8997 0>, rotation=<rot 90 0 0>}]
 calf_r [{shapeType=Capsule, length=25.7767, radius=11.8851, shapeName=calf_r_capsule, center=<22.58 1.5885 0>, rotation=<rot 90 0 0>}]
 dyn_sash_l_01 [{shapeType=Capsule, length=10.6709, radius=5.3071, shapeName=dyn_sash_l_01_capsule, center=<6.3819 0 3>, rotation=<rot 0 0 90>}]
 dyn_sash_l_02 [{shapeType=Capsule, length=10.6709, radius=5.3071, shapeName=dyn_sash_l_02_capsule, center=<6.3819 0 3>, rotation=<rot 0 0 90>}]
 dyn_sash_l_03 [{shapeType=Capsule, length=10.6709, radius=5.3071, shapeName=dyn_sash_l_03_capsule, center=<6.3819 0 3>, rotation=<rot 0 0 90>}]
 dyn_sash_l_mstr [{shapeType=Capsule, length=4, radius=4, shapeName=dyn_sash_l_mstr_capsule, rotation=<rot 0 0 90>}]
 dyn_sash_r_01 [{shapeType=Capsule, length=10.6709, radius=5.3071, shapeName=dyn_sash_r_01_capsule, center=<6.3819 0 3>, rotation=<rot 0 0 90>}]
 dyn_sash_r_02 [{shapeType=Capsule, length=10.6709, radius=5.3071, shapeName=dyn_sash_r_02_capsule, center=<6.3819 0 3>, rotation=<rot 0 0 90>}]
 dyn_sash_r_03 [{shapeType=Capsule, length=10.6709, radius=5.3071, shapeName=dyn_sash_r_03_capsule, center=<6.3819 0 3>, rotation=<rot 0 0 90>}]
 dyn_sash_r_mstr [{shapeType=Capsule, length=4, radius=4, shapeName=dyn_sash_r_mstr_capsule, rotation=<rot 0 0 90>}]
 dyn_skirt_bk_01 [{shapeType=Capsule, length=17.6632, radius=5.3071, shapeName=dyn_skirt_bk_01_capsule, center=<6.3819 0 3>, rotation=<rot 0 0 90>}]
 dyn_skirt_bk_02 [{shapeType=Capsule, length=19.1088, radius=11, shapeName=dyn_skirt_bk_02_capsule, center=<6.3819 0 3>, rotation=<rot 0 0 90>}]
 dyn_skirt_bk_03 [{shapeType=Capsule, length=24.5232, radius=11, shapeName=dyn_skirt_bk_03_capsule, center=<6.3819 0 3>, rotation=<rot 0 0 90>}]
 dyn_skirt_bk_04 [{shapeType=Capsule, length=20, radius=8, shapeName=dyn_skirt_bk_04_capsule, center=<6.3819 0 -3.0742>, rotation=<rot 0 0 90>}]
 dyn_skirt_bk_l_01 [{shapeType=Capsule, length=15.6738, radius=5.3071, shapeName=dyn_skirt_bk_l_01_capsule, center=<6.3819 0 3>, rotation=<rot 0 0 90>}]
 dyn_skirt_bk_l_02 [{shapeType=Capsule, length=18.5147, radius=11, shapeName=dyn_skirt_bk_l_02_capsule, center=<6.3819 0 3>, rotation=<rot 0 0 90>}]
 dyn_skirt_bk_l_03 [{shapeType=Capsule, length=18.6581, radius=11, shapeName=dyn_skirt_bk_l_03_capsule, center=<6.3819 0 3>, rotation=<rot 0 0 90>}]
 dyn_skirt_bk_l_mstr [{shapeType=Capsule, length=4, radius=4, shapeName=dyn_skirt_bk_l_mstr_capsule, rotation=<rot 0 0 90>}]
 dyn_skirt_bk_mstr [{shapeType=Capsule, length=4, radius=4, shapeName=dyn_skirt_bk_mstr_capsule, rotation=<rot 0 0 90>}]
 dyn_skirt_bk_r_01 [{shapeType=Capsule, length=15.6738, radius=5.3071, shapeName=dyn_skirt_bk_r_01_capsule, center=<6.3819 0 3>, rotation=<rot 0 0 90>}]
 dyn_skirt_bk_r_02 [{shapeType=Capsule, length=18.5147, radius=11, shapeName=dyn_skirt_bk_r_02_capsule, center=<6.3819 0 3>, rotation=<rot 0 0 90>}]
 dyn_skirt_bk_r_03 [{shapeType=Capsule, length=18.0748, radius=11, shapeName=dyn_skirt_bk_r_03_capsule, center=<6.3819 0 3>, rotation=<rot 0 0 90>}]
 dyn_skirt_bk_r_mstr [{shapeType=Capsule, length=4, radius=4, shapeName=dyn_skirt_bk_r_mstr_capsule, rotation=<rot 0 0 90>}]
 dyn_skirt_fr_l_01 [{shapeType=Capsule, length=14.9929, radius=5, shapeName=dyn_skirt_fr_l_01_capsule, center=<6.3819 0 4>, rotation=<rot 0 0 90>}]
 dyn_skirt_fr_l_02 [{shapeType=Capsule, length=16.8524, radius=6, shapeName=dyn_skirt_fr_r_02_capsule, center=<6.3819 0 4>, rotation=<rot 0 0 90>}]
 dyn_skirt_fr_l_03 [{shapeType=Capsule, length=16.8524, radius=4.5, shapeName=dyn_skirt_fr_r_02_capsule, center=<6.3819 0 3>, rotation=<rot 0 0 90>}]
 dyn_skirt_fr_l_mstr [{shapeType=Capsule, length=4, radius=4, shapeName=dyn_skirt_fr_l_mstr_capsule, rotation=<rot 0 0 90>}]
 dyn_skirt_fr_r_01 [{shapeType=Capsule, length=14.7756, radius=4.6843, shapeName=dyn_skirt_fr_r_01_capsule, center=<6.3819 0 3>, rotation=<rot 0 0 90>}]
 dyn_skirt_fr_r_02 [{shapeType=Capsule, length=16.1623, radius=6, shapeName=dyn_skirt_fr_r_02_capsule, center=<6.3819 0 3>, rotation=<rot 0 0 90>}]
 dyn_skirt_fr_r_03 [{shapeType=Capsule, length=15.4199, radius=6.4977, shapeName=dyn_skirt_fr_r_02_capsule, center=<6.3819 0 3>, rotation=<rot 0 0 90>}]
 dyn_skirt_fr_r_mstr [{shapeType=Capsule, length=4, radius=4, shapeName=dyn_skirt_fr_r_mstr_capsule, rotation=<rot 0 0 90>}]
 dyn_skirt_l_01 [{shapeType=Capsule, length=10.6709, radius=5.3071, shapeName=dyn_skirt_l_01_capsule, center=<6.3819 0 3>, rotation=<rot 0 0 90>}]
 dyn_skirt_l_02 [{shapeType=Capsule, length=10.6709, radius=5.3071, shapeName=dyn_skirt_l_02_capsule, center=<6.3819 0 3>, rotation=<rot 0 0 90>}]
 dyn_skirt_l_03 [{shapeType=Capsule, length=21.9643, radius=5.5891, shapeName=dyn_skirt_l_03_capsule, center=<-5.584 0 -3>, rotation=<rot 0 0 90>}]
 dyn_skirt_l_mstr [{shapeType=Capsule, length=4, radius=4, shapeName=dyn_skirt_l_mstr_capsule, rotation=<rot 0 0 90>}]
 dyn_skirt_r_01 [{shapeType=Capsule, length=10.6709, radius=5.3071, shapeName=dyn_skirt_r_01_capsule, center=<6.3819 0 3>, rotation=<rot 0 0 90>}]
 dyn_skirt_r_02 [{shapeType=Capsule, length=10.6709, radius=5.3071, shapeName=dyn_skirt_r_02_capsule, center=<6.3819 0 3>, rotation=<rot 0 0 90>}]
 dyn_skirt_r_03 [{shapeType=Capsule, length=21.8756, radius=5.6778, shapeName=dyn_skirt_r_03_capsule, center=<6.3819 0 2.1316>, rotation=<rot 0 0 90>}]
 dyn_skirt_r_mstr [{shapeType=Capsule, length=4, radius=4, shapeName=dyn_skirt_r_mstr_capsule, rotation=<rot 0 0 90>}]
 foot_l [{shapeType=Capsule, length=8.0331, radius=11.8617, shapeName=foot_l_capsule, center=<-2.5828 6.1491 0>, rotation=<rot 0 0 90>}]
 foot_r [{shapeType=Capsule, length=8.0224, radius=12.8743, shapeName=foot_r_capsule, center=<2.6557 -6.7129 0>, rotation=<rot 0 0 90>}]
 pelvis [{shapeType=Capsule, length=9.0849, radius=9.4281}]
 root [{shapeType=Capsule, length=1, radius=300, shapeName=root_capsule, center=<0 0 -300>, rotation=<rot 0 -90 0>}]
 thigh_l [
  {shapeType=Capsule, length=5.2589, radius=10.5, shapeName=thigh_l_capsule, center=<-21.8141 -0.6005 0.9807>, rotation=<rot 90 0 0>}
  {shapeType=Capsule, length=0.7604, radius=10.4804, shapeName=thigh_l_capsule, center=<-35.2106 -0.5871 0>, rotation=<rot 90 0 0>}
  {shapeType=Capsule, length=2.4897, radius=8.1772, shapeName=thigh_l_capsule, center=<-9.2295 -0.6005 0.7633>, rotation=<rot 0 0 90>}
 ]
 thigh_r [
  {shapeType=Capsule, length=8.7161, radius=10.5, shapeName=thigh_l_capsule, center=<25.0433 0.5002 -0.6389>, rotation=<rot 90 0 0>}
  {shapeType=Capsule, length=0.7604, radius=10.4804, shapeName=thigh_l_capsule, center=<35.6483 0.4924 0>, rotation=<rot 90 0 0>}
  {shapeType=Capsule, length=10.6481, radius=7.8242, shapeName=thigh_r_capsule, center=<10.003 0.5002 -0.9993>, rotation=<rot 90 0 0>}
 ]
}
constraints [
 {bone1Name=dyn_skirt_r_03, bone2Name=dyn_skirt_r_02, swing1LimitDegrees=5, swing2LimitDegrees=30, twistLimitDegrees=2}
 {
  bone1Name=dyn_sash_l_mstr
  bone2Name=pelvis
  swing1LimitDegrees=45
  swing1Motion=Free
  swing2LimitDegrees=45
  swing2Motion=Free
  twistLimitDegrees=45
  twistMotion=Free
 }
 {bone1Name=dyn_sash_l_01, bone2Name=dyn_sash_l_mstr, swing1LimitDegrees=20, swing2LimitDegrees=30, twistLimitDegrees=15}
 {bone1Name=dyn_sash_l_02, bone2Name=dyn_sash_l_01, swing1LimitDegrees=20, swing2LimitDegrees=30, twistLimitDegrees=15}
 {bone1Name=dyn_sash_l_03, bone2Name=dyn_sash_l_02, swing1LimitDegrees=20, swing2LimitDegrees=30, twistLimitDegrees=15}
 {
  bone1Name=dyn_sash_r_mstr
  bone2Name=pelvis
  swing1LimitDegrees=45
  swing1Motion=Free
  swing2LimitDegrees=45
  swing2Motion=Free
  twistLimitDegrees=45
  twistMotion=Free
 }
 {bone1Name=dyn_sash_r_01, bone2Name=dyn_sash_r_mstr, swing1LimitDegrees=20, swing2LimitDegrees=30, twistLimitDegrees=15}
 {bone1Name=dyn_sash_r_02, bone2Name=dyn_sash_r_01, swing1LimitDegrees=20, swing2LimitDegrees=30, twistLimitDegrees=15}
 {bone1Name=dyn_sash_r_03, bone2Name=dyn_sash_r_02, swing1LimitDegrees=20, swing2LimitDegrees=30, twistLimitDegrees=15}
 {
  bone1Name=dyn_skirt_bk_l_mstr
  bone2Name=pelvis
  swing1LimitDegrees=45
  swing1Motion=Free
  swing2LimitDegrees=45
  swing2Motion=Free
  twistLimitDegrees=45
  twistMotion=Free
 }
 {bone1Name=dyn_skirt_bk_l_01, bone2Name=dyn_skirt_bk_l_mstr, swing1LimitDegrees=7, swing2LimitDegrees=30, twistLimitDegrees=6}
 {bone1Name=dyn_skirt_bk_l_02, bone2Name=dyn_skirt_bk_l_01, swing1LimitDegrees=10, swing2LimitDegrees=30, twistLimitDegrees=15}
 {bone1Name=dyn_skirt_bk_l_03, bone2Name=dyn_skirt_bk_l_02, swing1LimitDegrees=10, swing2LimitDegrees=30, twistLimitDegrees=15}
 {
  bone1Name=dyn_skirt_bk_mstr
  bone2Name=pelvis
  swing1LimitDegrees=45
  swing1Motion=Free
  swing2LimitDegrees=45
  swing2Motion=Free
  twistLimitDegrees=45
  twistMotion=Free
 }
 {bone1Name=dyn_skirt_bk_01, bone2Name=dyn_skirt_bk_mstr, swing1LimitDegrees=20, swing2LimitDegrees=30, twistLimitDegrees=7}
 {bone1Name=dyn_skirt_bk_02, bone2Name=dyn_skirt_bk_01, swing1LimitDegrees=20, swing2LimitDegrees=30, twistLimitDegrees=7}
 {bone1Name=dyn_skirt_bk_03, bone2Name=dyn_skirt_bk_02, swing1LimitDegrees=10, swing2LimitDegrees=30, twistLimitDegrees=7}
 {bone1Name=dyn_skirt_bk_04, bone2Name=dyn_skirt_bk_03, swing1LimitDegrees=10, swing2LimitDegrees=30, twistLimitDegrees=7}
 {
  bone1Name=dyn_skirt_bk_r_mstr
  bone2Name=pelvis
  swing1LimitDegrees=45
  swing1Motion=Free
  swing2LimitDegrees=45
  swing2Motion=Free
  twistLimitDegrees=45
  twistMotion=Free
 }
 {bone1Name=dyn_skirt_bk_r_01, bone2Name=dyn_skirt_bk_r_mstr, swing1LimitDegrees=7, swing2LimitDegrees=30, twistLimitDegrees=6}
 {bone1Name=dyn_skirt_bk_r_02, bone2Name=dyn_skirt_bk_r_01, swing1LimitDegrees=7, swing2LimitDegrees=30, twistLimitDegrees=6}
 {bone1Name=dyn_skirt_bk_r_03, bone2Name=dyn_skirt_bk_r_02, swing1LimitDegrees=7, swing2LimitDegrees=30, twistLimitDegrees=6}
 {
  bone1Name=dyn_skirt_fr_l_mstr
  bone2Name=pelvis
  swing1LimitDegrees=45
  swing1Motion=Free
  swing2LimitDegrees=45
  swing2Motion=Free
  twistLimitDegrees=45
  twistMotion=Free
 }
 {bone1Name=dyn_skirt_fr_l_01, bone2Name=dyn_skirt_fr_l_mstr, swing1LimitDegrees=10, swing2LimitDegrees=30, twistLimitDegrees=12}
 {bone1Name=dyn_skirt_fr_l_02, bone2Name=dyn_skirt_fr_l_01, swing1LimitDegrees=20, swing2LimitDegrees=30, twistLimitDegrees=15}
 {
  bone1Name=dyn_skirt_fr_r_mstr
  bone2Name=pelvis
  swing1LimitDegrees=45
  swing1Motion=Free
  swing2LimitDegrees=45
  swing2Motion=Free
  twistLimitDegrees=45
  twistMotion=Free
 }
 {bone1Name=dyn_skirt_fr_r_01, bone2Name=dyn_skirt_fr_r_mstr, swing1LimitDegrees=10, swing2LimitDegrees=30, twistLimitDegrees=7}
 {bone1Name=dyn_skirt_fr_r_02, bone2Name=dyn_skirt_fr_r_01, swing1LimitDegrees=10, swing2LimitDegrees=30, twistLimitDegrees=7}
 {
  bone1Name=dyn_skirt_l_mstr
  bone2Name=pelvis
  swing1LimitDegrees=45
  swing1Motion=Free
  swing2LimitDegrees=45
  swing2Motion=Free
  twistLimitDegrees=45
  twistMotion=Free
 }
 {bone1Name=dyn_skirt_l_01, bone2Name=dyn_skirt_l_mstr, swing1LimitDegrees=14, swing2LimitDegrees=30, twistLimitDegrees=10}
 {bone1Name=dyn_skirt_l_02, bone2Name=dyn_skirt_l_01, swing1LimitDegrees=14, swing2LimitDegrees=30, twistLimitDegrees=10}
 {bone1Name=dyn_skirt_l_03, bone2Name=dyn_skirt_l_02, swing1LimitDegrees=14, swing2LimitDegrees=30, twistLimitDegrees=10}
 {
  bone1Name=dyn_skirt_r_mstr
  bone2Name=pelvis
  swing1LimitDegrees=45
  swing1Motion=Free
  swing2LimitDegrees=45
  swing2Motion=Free
  twistLimitDegrees=45
  twistMotion=Free
 }
 {bone1Name=dyn_skirt_r_01, bone2Name=dyn_skirt_r_mstr, swing1LimitDegrees=5, swing2LimitDegrees=30, twistLimitDegrees=2}
 {bone1Name=dyn_skirt_r_02, bone2Name=dyn_skirt_r_01, swing1LimitDegrees=5, swing2LimitDegrees=30, twistLimitDegrees=2}
 {
  bone1Name=thigh_l
  bone2Name=pelvis
  swing1LimitDegrees=45
  swing1Motion=Free
  swing2LimitDegrees=45
  swing2Motion=Free
  twistLimitDegrees=45
  twistMotion=Free
 }
 {
  bone1Name=thigh_r
  bone2Name=pelvis
  swing1LimitDegrees=45
  swing1Motion=Free
  swing2LimitDegrees=45
  swing2Motion=Free
  twistLimitDegrees=45
  twistMotion=Free
 }
 {
  bone1Name=calf_l
  bone2Name=thigh_l
  swing1LimitDegrees=45
  swing1Motion=Free
  swing2LimitDegrees=45
  swing2Motion=Free
  twistLimitDegrees=45
  twistMotion=Free
 }
 {
  bone1Name=calf_r
  bone2Name=thigh_r
  swing1LimitDegrees=45
  swing1Motion=Free
  swing2LimitDegrees=45
  swing2Motion=Free
  twistLimitDegrees=45
  twistMotion=Free
 }
 {
  bone1Name=foot_r
  bone2Name=calf_r
  swing1LimitDegrees=45
  swing1Motion=Free
  swing2LimitDegrees=45
  swing2Motion=Free
  twistLimitDegrees=45
  twistMotion=Free
 }
 {
  bone1Name=foot_l
  bone2Name=calf_l
  swing1LimitDegrees=45
  swing1Motion=Free
  swing2LimitDegrees=45
  swing2Motion=Free
  twistLimitDegrees=45
  twistMotion=Free
 }
 {bone1Name=dyn_skirt_fr_r_03, bone2Name=dyn_skirt_fr_r_02, swing1LimitDegrees=10, swing2LimitDegrees=30, twistLimitDegrees=7}
 {bone1Name=dyn_skirt_fr_l_03, bone2Name=dyn_skirt_fr_l_02, swing1LimitDegrees=10, swing2LimitDegrees=30, twistLimitDegrees=10}
 {
  bone1Name=dyn_skirt_fr_l_03
  bone2Name=dyn_skirt_l_03
  swing1LimitDegrees=45
  swing1Motion=Free
  swing2LimitDegrees=45
  swing2Motion=Free
  twistLimitDegrees=45
  twistMotion=Free
 }
 {
  bone1Name=pelvis
  bone2Name=root
  swing1LimitDegrees=45
  swing1Motion=Free
  swing2LimitDegrees=45
  swing2Motion=Free
  twistLimitDegrees=45
  twistMotion=Free
 }
]
