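# Post-process Control Rig for the female medium PlantStand Cane body that corrects skirt and flap bones as the thighs move.
# Spherical pose readers measure thigh poses relative to spine_01, driving weights for parent constraints and transform offsets.
# Left-leg raise drives right middle flap bones and vice versa; extreme right-leg raise, reduced by crouching or fixed by r_leg_lock, pulls right skirt chains toward the thigh twist.
# Two dyn_flap_mid constraints are unconnected.
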ControlRigBlueprint /BRCosmetics/Characters/Player/Female/Medium/Bodies/F_MED_PlantStand_Cane/Meshes/F_MED_PlantStand_Cane_Post_CtrlRig
graphs [
 RigVMFunctionLibrary {links [], nodes []}
 RigVMModel {nodes [
   RigUnit_BeginExecution = &RigUnit_BeginExecution()
   RigVMFunction_Sequence = &RigVMFunction_Sequence(ExecuteContext=@RigUnit_BeginExecution.ExecutePin)
   n1 = &RigUnit_SphericalPoseReader(
    ActiveRegionScaleFactors=(PositiveWidth=3.5,NegativeWidth=0.5,PositiveHeight=4,NegativeHeight=4)
    ActiveRegionSize=0.1
    Debug=(bDrawDebug=false,DebugScale=25,DebugSegments=20,DebugThickness=0.25)
    DriverAxis=(X=-1,Y=0,Z=0)
    DriverItem=(Type=Bone,Name="thigh_l")
    FalloffRegionScaleFactors=(PositiveWidth=1,NegativeWidth=1,PositiveHeight=1,NegativeHeight=1)
    FalloffSize=0.45
    OptionalParentItem=(Type=Bone,Name="spine_01")
    RotationOffset=(X=105,Y=0,Z=0)
    ExecutePin=@RigVMFunction_Sequence.A
   )
   > RigVMVariableNode(Value=@n1.OutputParam) {variable=l_leg_up}
   > n2 = &RigUnit_SphericalPoseReader(
    ActiveRegionScaleFactors=(PositiveWidth=3.5,NegativeWidth=0.5,PositiveHeight=3.5,NegativeHeight=3)
    ActiveRegionSize=0.1
    Debug=(bDrawDebug=false,DebugScale=25,DebugSegments=20,DebugThickness=0.25)
    DriverAxis=(X=1,Y=0,Z=0)
    DriverItem=(Type=Bone,Name="thigh_r")
    FalloffRegionScaleFactors=(PositiveWidth=1,NegativeWidth=1,PositiveHeight=1,NegativeHeight=1)
    FalloffSize=0.45
    FlipHeightScaling=false
    FlipWidthScaling=true
    OptionalParentItem=(Type=Bone,Name="spine_01")
    RotationOffset=(X=-90,Y=0,Z=15)
   )
   > RigVMVariableNode(Value=@n2.OutputParam) {variable=r_leg_up}
   > n3 = &RigUnit_SphericalPoseReader(
    ActiveRegionScaleFactors=(PositiveWidth=4.5,NegativeWidth=0.3,PositiveHeight=4,NegativeHeight=4)
    ActiveRegionSize=0.1
    Debug=(bDrawDebug=false,DebugScale=25,DebugSegments=20,DebugThickness=0.25)
    DriverAxis=(X=1,Y=0,Z=0)
    DriverItem=(Type=Bone,Name="thigh_r")
    FalloffRegionScaleFactors=(PositiveWidth=1,NegativeWidth=0.3,PositiveHeight=1,NegativeHeight=1)
    FalloffSize=0.45
    FlipHeightScaling=false
    FlipWidthScaling=true
    OptionalParentItem=(Type=Bone,Name="spine_01")
    RotationOffset=(X=-90,Y=0,Z=-50)
   )
   RigVMFunction_ControlFlowBranch = &RigVMFunction_ControlFlowBranch(ExecuteContext=@n3.ExecutePin, Condition=RigVMVariableNode() {variable=r_leg_lock}.Value)
   Clamp = &RigVMFunction_MathDoubleClamp(Value=&RigVMFunction_MathDoubleSub(A=@n3.OutputParam, B=RigVMVariableNode() {variable=crouching}.Value).Result)
   RigVMVariableNode(Value=@Clamp.Result, ExecuteContext=@RigVMFunction_ControlFlowBranch.False) {variable=r_leg_up_extreme}
   RigVMVariableNode(Value=0.6, ExecuteContext=@RigVMFunction_ControlFlowBranch.True) {variable=r_leg_up_extreme}
   n4 = RigVMVariableNode() {variable=l_leg_up}
   &RigUnit_ParentConstraint(
    AdvancedSettings=(InterpolationType=Average,RotationOrderForFilter=XZY)
    Child=(Type=Bone,Name="dyn_flap_mid_r_01")
    Parents="((Item=(Type=Bone,Name=\"pelvis\"),Weight=0.37),(Item=(Type=Bone,Name=\"dyn_flap_l_02\"),Weight=0),(Item=(Type=Bone,Name=\"dyn_flap_r_01\"),Weight=0.01),(Item=(Type=Bone,Name=\"dyn_flap_l_01\"),Weight=0.03),(Item=(Type=Bone,Name=\"dyn_flap_fr_r_01\"),Weight=0.34),(I ...(truncated)"
    ExecutePin=@RigVMFunction_Sequence.B
    Weight=@n4.Value
   )
   > &RigUnit_ModifyTransforms(
    ItemToModify=((Item=(Type=Bone,Name="dyn_flap_mid_r_01"),Transform=(Rotation=(X=-0.1261,Y=-0.2566,Z=-0.0338,W=0.9577),Translation=(X=0,Y=0,Z=1),Scale3D=(X=1,Y=1,Z=1))))
    Weight=1
    WeightMinimum=0
   )
   > &RigUnit_ParentConstraint(
    AdvancedSettings=(InterpolationType=Average,RotationOrderForFilter=XZY)
    Child=(Type=Bone,Name="dyn_flap_mid_r_02")
    Parents="((Item=(Type=Bone,Name=\"pelvis\"),Weight=0),(Item=(Type=Bone,Name=\"dyn_flap_fr_r_02\"),Weight=0.39),(Item=(Type=Bone,Name=\"dyn_flap_fr_r_01\"),Weight=0.1),(Item=(Type=Bone,Name=\"dyn_flap_fr_l_02\"),Weight=0.17),(Item=(Type=Bone,Name=\"dyn_flap_fr_l_01\"),Weight= ...(truncated)"
    Weight=@n4.Value
   )
   > &RigUnit_ModifyTransforms(
    ItemToModify=((Item=(Type=Bone,Name="dyn_flap_mid_r_02"),Transform=(Rotation=(X=-0.7064,Y=-0.0308,Z=0.0308,W=0.7064),Translation=(X=0,Y=4.5,Z=-0.3),Scale3D=(X=1,Y=1,Z=1))))
    WeightMinimum=0
    Weight=@n4.Value
   )
   n5 = RigVMVariableNode() {variable=r_leg_up}
   > &RigUnit_ParentConstraint(
    AdvancedSettings=(InterpolationType=Average,RotationOrderForFilter=XZY)
    Child=(Type=Bone,Name="dyn_flap_mid_l_01")
    Parents="((Item=(Type=Bone,Name=\"pelvis\"),Weight=0.37),(Item=(Type=Bone,Name=\"dyn_flap_l_02\"),Weight=0),(Item=(Type=Bone,Name=\"dyn_flap_r_01\"),Weight=0.01),(Item=(Type=Bone,Name=\"dyn_flap_l_01\"),Weight=0.03),(Item=(Type=Bone,Name=\"dyn_flap_fr_r_01\"),Weight=0.34),(I ...(truncated)"
    Weight=@n5.Value
   )
   > &RigUnit_ParentConstraint(
    AdvancedSettings=(InterpolationType=Average,RotationOrderForFilter=XZY)
    Child=(Type=Bone,Name="dyn_flap_mid_l_02")
    Parents="((Item=(Type=Bone,Name=\"pelvis\"),Weight=0),(Item=(Type=Bone,Name=\"dyn_flap_fr_r_02\"),Weight=0.39),(Item=(Type=Bone,Name=\"dyn_flap_fr_r_01\"),Weight=0.1),(Item=(Type=Bone,Name=\"dyn_flap_fr_l_02\"),Weight=0.17),(Item=(Type=Bone,Name=\"dyn_flap_fr_l_01\"),Weight= ...(truncated)"
    Weight=@n5.Value
   )
   > &RigUnit_ModifyTransforms(
    ItemToModify=((Item=(Type=Bone,Name="dyn_flap_mid_l_02"),Transform=(Rotation=(X=0.2588,Y=0,Z=0,W=0.9659),Translation=(X=0,Y=0,Z=-0.3),Scale3D=(X=1,Y=1,Z=1))))
    WeightMinimum=0
    Weight=@n5.Value
   )
   n6 = RigVMVariableNode() {variable=r_leg_up_extreme}
   > &RigUnit_ParentConstraint(
    AdvancedSettings=(InterpolationType=Average,RotationOrderForFilter=XZY)
    Child=(Type=Bone,Name="dyn_skirt_fr_r_01")
    Filter=(TranslationFilter=(bX=true,bY=true,bZ=true),RotationFilter=(bX=True,bY=True,bZ=True),ScaleFilter=(bX=True,bY=True,bZ=True))
    Parents=((Item=(Type=Bone,Name="thigh_twist_01_r"),Weight=1))
    Weight=@n6.Value
   )
   > &RigUnit_ParentConstraint(
    AdvancedSettings=(InterpolationType=Average,RotationOrderForFilter=XZY)
    Child=(Type=Bone,Name="dyn_skirt_fr_r_02")
    Filter=(TranslationFilter=(bX=true,bY=true,bZ=true),RotationFilter=(bX=True,bY=True,bZ=True),ScaleFilter=(bX=True,bY=True,bZ=True))
    Parents=((Item=(Type=Bone,Name="thigh_twist_01_r"),Weight=1))
    Weight=@n6.Value
   )
   > &RigUnit_ParentConstraint(
    AdvancedSettings=(InterpolationType=Average,RotationOrderForFilter=XZY)
    Child=(Type=Bone,Name="dyn_skirt_fr_r_03")
    Filter=(TranslationFilter=(bX=true,bY=true,bZ=true),RotationFilter=(bX=True,bY=True,bZ=True),ScaleFilter=(bX=True,bY=True,bZ=True))
    Parents=((Item=(Type=Bone,Name="thigh_twist_01_r"),Weight=1))
    Weight=@n6.Value
   )
   > &RigUnit_ParentConstraint(
    AdvancedSettings=(InterpolationType=Average,RotationOrderForFilter=XZY)
    Child=(Type=Bone,Name="dyn_skirt_r_01")
    Filter=(TranslationFilter=(bX=true,bY=true,bZ=true),RotationFilter=(bX=True,bY=True,bZ=True),ScaleFilter=(bX=True,bY=True,bZ=True))
    Parents=((Item=(Type=Bone,Name="thigh_twist_01_r"),Weight=1))
    Weight=@n6.Value
   )
   > &RigUnit_ParentConstraint(
    AdvancedSettings=(InterpolationType=Average,RotationOrderForFilter=XZY)
    Child=(Type=Bone,Name="dyn_skirt_r_02")
    Filter=(TranslationFilter=(bX=true,bY=true,bZ=true),RotationFilter=(bX=True,bY=True,bZ=True),ScaleFilter=(bX=True,bY=True,bZ=True))
    Parents=((Item=(Type=Bone,Name="thigh_twist_01_r"),Weight=1))
    Weight=@n6.Value
   )
   > &RigUnit_ParentConstraint(
    AdvancedSettings=(InterpolationType=Average,RotationOrderForFilter=XZY)
    Child=(Type=Bone,Name="dyn_skirt_r_03")
    Filter=(TranslationFilter=(bX=true,bY=true,bZ=true),RotationFilter=(bX=True,bY=True,bZ=True),ScaleFilter=(bX=True,bY=True,bZ=True))
    Parents=((Item=(Type=Bone,Name="thigh_twist_01_r"),Weight=1))
    Weight=@n6.Value
   )
   > &RigUnit_ModifyTransforms(
    ItemToModify=((Item=(Type=Bone,Name="dyn_skirt_fr_r_01"),Transform=(Rotation=(X=0,Y=-0.0436,Z=0,W=0.999),Translation=(X=0,Y=0,Z=0),Scale3D=(X=1,Y=1,Z=1))))
    WeightMinimum=0
    Weight=@n6.Value
   )
   &RigUnit_ParentConstraint(
    AdvancedSettings=(InterpolationType=Average,RotationOrderForFilter=XZY)
    Child=(Type=Bone,Name="dyn_flap_mid_01")
    ExecutePin=()
    Parents=((Item=(Type=Bone,Name="dyn_flap_fr_r_02"),Weight=1))
    Weight=1
   )
   &RigUnit_ParentConstraint(
    AdvancedSettings=(InterpolationType=Average,RotationOrderForFilter=XZY)
    Child=(Type=Bone,Name="dyn_flap_mid_02")
    ExecutePin=()
    Parents=((Item=(Type=Bone,Name="dyn_skirt_fr_r_03"),Weight=1))
    Weight=1
   )
   RigVMVariableNode() {variable=l_leg_up}
  ]}
]
hierarchy [
 root {loc=<0 0 0>, rot=<0 0 0 1>}
 pelvis {parent=root, loc=<0 -0.5301 98.0571>, rot=<0.0127 0.707 -0.0127 -0.707>}
 spine_01 {parent=pelvis, loc=<6.1208 -0.9328 0>, rot=<0 0 0.034 -0.9994>}
 spine_02 {parent=spine_01, loc=<6.5996 -0.2095 0>, rot=<0 0 0.0238 -0.9997>}
 spine_03 {parent=spine_02, loc=<6.4914 -0.1237 0>, rot=<0 0 0.0149 -0.9999>}
 spine_04 {parent=spine_03, loc=<6.9598 0.0794 0>, rot=<0 0 -0.0132 -0.9999>}
 spine_05 {parent=spine_04, loc=<6.9987 0 0>, rot=<0 0 -0.0015 -1>}
 attach_cape {parent=spine_05, loc=<9.3216 -8.4501 0>, rot=<0.0616 -0.7044 -0.0616 -0.7044>}
 neck_01 {parent=spine_05, loc=<11.3138 0.1529 0>, rot=<0 0 -0.2003 -0.9797>}
 neck_02 {parent=neck_01, loc=<4.5627 0 0>, rot=<0 0 0.0504 -0.9987>}
 head {parent=neck_02, loc=<5.3914 0 0>, rot=<0 0 0.115 -0.9934>}
 faceAttach {parent=head, loc=<-0.3783 -1.1047 -0.0136>, rot=<0.0029 -0.7071 -0.0029 -0.7071>}
 C_jaw {parent=faceAttach, loc=<0.0319 1.3051 -0.8738>, rot=<0 0 0 1>}
 attach_backpack {parent=spine_05, loc=<0 -8.35 0>, rot=<-0.0093 -0.707 0.0093 -0.707>}
 clavicle_l {parent=spine_05, loc=<5.4831 -0.1053 -3.2946>, rot=<0.0465 -0.7289 -0.009 -0.6829>}
 upperarm_l {parent=clavicle_l, loc=<9.8658 -1.2136 -0.4105>, rot=<0.0164 -0.3245 -0.0269 -0.9454>}
 lowerarm_l {parent=upperarm_l, loc=<26.465 -0.8345 0.259>, rot=<0.0036 0.0161 -0.2298 -0.9731>}
 hand_l {parent=lowerarm_l, loc=<22.7355 -0.3203 -0.6105>, rot=<-0.6784 -0.0223 0.0735 -0.7307>}
 thumb_01_l {parent=hand_l, loc=<1.3118 -0.3191 -2.459>, rot=<0.6546 -0.4042 0.0664 -0.6354>}
 thumb_02_l {parent=thumb_01_l, loc=<5.9336 -0.0448 0.0401>, rot=<-0.0037 0.0213 0.298 -0.9543>}
 thumb_03_l {parent=thumb_02_l, loc=<2.2223 0.3682 -0.0907>, rot=<0.0118 -0.013 0.0066 -0.9998>}
 index_metacarpal_l {parent=hand_l, loc=<1.6535 0.0183 -2.96>, rot=<0.0158 -0.0124 -0.0173 -0.9996>}
 index_01_l {parent=index_metacarpal_l, loc=<8.3008 0.0357 -0.0524>, rot=<-0.0044 0.0485 0.0924 -0.9945>}
 index_02_l {parent=index_01_l, loc=<3.3587 -0.3004 0.0452>, rot=<-0.002 0.0027 0.3225 -0.9466>}
 index_03_l {parent=index_02_l, loc=<2.5578 -0.123 -0.0252>, rot=<-0.002 0.0067 0.2558 -0.9667>}
 middle_metacarpal_l {parent=hand_l, loc=<2.1856 -0.1725 -0.5984>, rot=<0.006 0.0168 -0.0329 -0.9993>}
 middle_01_l {parent=middle_metacarpal_l, loc=<7.3275 -0.2447 0.0129>, rot=<-0.0208 0.0042 0.0858 -0.9961>}
 middle_02_l {parent=middle_01_l, loc=<4.0123 0.1401 0.1175>, rot=<-0.0218 0.0258 0.3512 -0.9357>}
 middle_03_l {parent=middle_02_l, loc=<2.6077 0.0534 0.1373>, rot=<0.0082 0.0188 0.1708 -0.9851>}
 ring_metacarpal_l {parent=hand_l, loc=<2.2074 -0.3625 1.3638>, rot=<-0.0284 0.0722 -0.0458 -0.9959>}
 ring_01_l {parent=ring_metacarpal_l, loc=<7.0295 -0.6197 -0.1446>, rot=<-0.0489 -0.0252 0.1892 -0.9804>}
 ring_02_l {parent=ring_01_l, loc=<3.6174 0.4647 -0.1027>, rot=<0 0 0.2806 -0.9598>}
 ring_03_l {parent=ring_02_l, loc=<2.4152 0.114 0.0199>, rot=<0 0 0.1189 -0.9929>}
 pinky_metacarpal_l {parent=hand_l, loc=<1.7616 -0.9753 3.3814>, rot=<-0.0397 0.0814 -0.069 -0.9935>}
 pinky_01_l {parent=pinky_metacarpal_l, loc=<7.1709 -0.4831 -0.1094>, rot=<-0.0654 -0.0035 0.2242 -0.9723>}
 pinky_02_l {parent=pinky_01_l, loc=<2.8274 0.0893 -0.1043>, rot=<0.0037 0.0011 0.2854 -0.9584>}
 pinky_03_l {parent=pinky_02_l, loc=<2.2046 0.0307 -0.0078>, rot=<-0.0002 0.0098 0.1445 -0.9895>}
 lowerarm_twist_02_l {parent=lowerarm_l, loc=<10.6707 0 0>, rot=<0 0 0 1>}
 lowerarm_twist_01_l {parent=lowerarm_l, loc=<18.316 0 0>, rot=<0 0 0 1>}
 upperarm_twist_01_l {parent=upperarm_l, loc=<7.2135 -0.1134 0>, rot=<0 0 0 1>}
 upperarm_twist_02_l {parent=upperarm_l, loc=<19.7998 -0.1134 0>, rot=<0 0 0 1>}
 clavicle_r {parent=spine_05, loc=<5.4831 -0.1053 3.2946>, rot=<0.7289 0.0465 -0.6829 0.009>}
 upperarm_r {parent=clavicle_r, loc=<-9.8658 1.2136 0.4105>, rot=<0.0164 -0.3245 -0.0269 -0.9454>}
 lowerarm_r {parent=upperarm_r, loc=<-26.465 0.8345 -0.259>, rot=<0.0036 0.0161 -0.2298 -0.9731>}
 hand_r {parent=lowerarm_r, loc=<-22.7355 0.3203 0.6106>, rot=<-0.6784 -0.0223 0.0735 -0.7307>}
 thumb_01_r {parent=hand_r, loc=<-1.3118 0.3191 2.459>, rot=<0.6546 -0.4042 0.0664 -0.6354>}
 thumb_02_r {parent=thumb_01_r, loc=<-5.9336 0.0448 -0.0401>, rot=<-0.0037 0.0213 0.298 -0.9543>}
 thumb_03_r {parent=thumb_02_r, loc=<-2.2223 -0.3682 0.0907>, rot=<0.0118 -0.013 0.0066 -0.9998>}
 index_metacarpal_r {parent=hand_r, loc=<-1.6535 -0.0183 2.96>, rot=<0.0158 -0.0124 -0.0173 -0.9996>}
 index_01_r {parent=index_metacarpal_r, loc=<-8.3008 -0.0357 0.0524>, rot=<-0.0044 0.0485 0.0924 -0.9945>}
 index_02_r {parent=index_01_r, loc=<-3.3587 0.3004 -0.0452>, rot=<-0.002 0.0027 0.3225 -0.9466>}
 index_03_r {parent=index_02_r, loc=<-2.5578 0.123 0.0252>, rot=<-0.002 0.0067 0.2558 -0.9667>}
 middle_metacarpal_r {parent=hand_r, loc=<-2.1856 0.1725 0.5984>, rot=<0.006 0.0168 -0.0329 -0.9993>}
 middle_01_r {parent=middle_metacarpal_r, loc=<-7.3275 0.2447 -0.0129>, rot=<-0.0208 0.0042 0.0858 -0.9961>}
 middle_02_r {parent=middle_01_r, loc=<-4.0123 -0.1401 -0.1175>, rot=<-0.0218 0.0258 0.3512 -0.9357>}
 middle_03_r {parent=middle_02_r, loc=<-2.6077 -0.0534 -0.1373>, rot=<0.0082 0.0188 0.1708 -0.9851>}
 ring_metacarpal_r {parent=hand_r, loc=<-2.2074 0.3625 -1.3638>, rot=<-0.0284 0.0722 -0.0458 -0.9959>}
 ring_01_r {parent=ring_metacarpal_r, loc=<-7.0295 0.6197 0.1446>, rot=<-0.0489 -0.0252 0.1892 -0.9804>}
 ring_02_r {parent=ring_01_r, loc=<-3.6174 -0.4647 0.1027>, rot=<0 0 0.2806 -0.9598>}
 ring_03_r {parent=ring_02_r, loc=<-2.4152 -0.114 -0.0199>, rot=<0 0 0.1189 -0.9929>}
 pinky_metacarpal_r {parent=hand_r, loc=<-1.7616 0.9753 -3.3814>, rot=<-0.0397 0.0814 -0.069 -0.9935>}
 pinky_01_r {parent=pinky_metacarpal_r, loc=<-7.1709 0.4831 0.1094>, rot=<-0.0654 -0.0035 0.2242 -0.9723>}
 pinky_02_r {parent=pinky_01_r, loc=<-2.8274 -0.0893 0.1043>, rot=<0.0037 0.0011 0.2854 -0.9584>}
 pinky_03_r {parent=pinky_02_r, loc=<-2.2046 -0.0307 0.0078>, rot=<-0.0002 0.0098 0.1445 -0.9895>}
 lowerarm_twist_02_r {parent=lowerarm_r, loc=<-10.6707 0 0>, rot=<0 0 0 -1>}
 lowerarm_twist_01_r {parent=lowerarm_r, loc=<-18.316 0 0>, rot=<0 0 0 -1>}
 upperarm_twist_01_r {parent=upperarm_r, loc=<-7.2135 0.1134 0>, rot=<0 0 0 -1>}
 upperarm_twist_02_r {parent=upperarm_r, loc=<-19.7998 0.1134 0>, rot=<0 0 0 -1>}
 thigh_l {parent=pelvis, loc=<-1.0855 -0.2109 -8.7053>, rot=<0.0009 0.0355 0.0155 -0.9993>}
 calf_l {parent=thigh_l, loc=<-42.5284 -0.0019 -0.7297>, rot=<0.0006 -0.0161 -0.0391 -0.9991>}
 foot_l {parent=calf_l, loc=<-46.7401 -2.5014 -0.592>, rot=<0.0427 -0.0238 0.0353 -0.9982>}
 ball_l {parent=foot_l, loc=<-3.5658 10.8066 -0.1544>, rot=<0.0001 -0.0001 -0.7063 -0.7079>}
 calf_twist_02_l {parent=calf_l, loc=<-19.1546 -1.7943 -0.2638>, rot=<0.0237 0.0095 -0.0035 -0.9997>}
 calf_twist_01_l {parent=calf_l, loc=<-35.8483 -1.9197 -0.5794>, rot=<0.0237 0.0095 -0.0035 -0.9997>}
 dyn_pants_l_mstr {parent=calf_l, loc=<-24.103 -1.5607 -0.2853>, rot=<-0.0309 -0.7201 0.0276 -0.6927>}
 dyn_pants_l {parent=dyn_pants_l_mstr, loc=<0 0 0>, rot=<0 0 0 1>}
 thigh_twist_01_l {parent=thigh_l, loc=<-27.952 0 0>, rot=<0.0013 0 0 -1>}
 thigh_r {parent=pelvis, loc=<-1.0855 -0.2109 8.7053>, rot=<0.0355 -0.0009 0.9993 0.0155>}
 calf_r {parent=thigh_r, loc=<42.5284 0.0019 0.7297>, rot=<0.0006 -0.0161 -0.0391 -0.9991>}
 foot_r {parent=calf_r, loc=<46.7401 2.5014 0.592>, rot=<0.0427 -0.0238 0.0353 -0.9982>}
 ball_r {parent=foot_r, loc=<3.5658 -10.8066 0.1544>, rot=<0.0001 -0.0001 -0.7063 -0.7079>}
 calf_twist_02_r {parent=calf_r, loc=<19.1546 1.7943 0.2638>, rot=<0.0237 0.0095 -0.0035 -0.9997>}
 calf_twist_01_r {parent=calf_r, loc=<35.8483 1.9197 0.5794>, rot=<0.0237 0.0095 -0.0035 -0.9997>}
 dyn_pants_r_mstr {parent=calf_r, loc=<24.0289 0.6347 0.3564>, rot=<-0.6927 0.0276 0.7201 0.0309>}
 dyn_pants_r {parent=dyn_pants_r_mstr, loc=<0 0 0>, rot=<0 0 0 1>}
 thigh_twist_01_r {parent=thigh_r, loc=<27.952 0 0>, rot=<0.0013 0 0 -1>}
 skirt_base {parent=pelvis, loc=<0 0 0>, rot=<0 0 0 1>}
 dyn_skirt_bk_mstr {parent=skirt_base, loc=<2.8077 -11.633 0>, rot=<0.0232 -0.7059 0.707 -0.0369>}
 dyn_skirt_bk_01 {parent=dyn_skirt_bk_mstr, loc=<0 0 0>, rot=<0 0 0 1>}
 dyn_skirt_bk_02 {parent=dyn_skirt_bk_01, loc=<10.4476 0 0>, rot=<0.0009 -0.0123 0.0083 -0.9999>}
 dyn_skirt_bk_03 {parent=dyn_skirt_bk_02, loc=<11.5802 0 0>, rot=<0.0004 -0.0077 0.0044 -1>}
 dyn_skirt_bk_04 {parent=dyn_skirt_bk_03, loc=<10.7435 0 0>, rot=<-0.0004 0.4722 -0.0006 -0.8815>}
 dyn_skirt_bk_r_mstr {parent=skirt_base, loc=<3.5231 -9.102 8.7892>, rot=<0.121 -0.5202 0.8452 -0.0193>}
 dyn_skirt_bk_r_01 {parent=dyn_skirt_bk_r_mstr, loc=<0 0 0>, rot=<0 0 0 1>}
 dyn_skirt_bk_r_02 {parent=dyn_skirt_bk_r_01, loc=<10.7603 0 0>, rot=<0.0009 -0.0933 -0.0207 -0.9954>}
 dyn_skirt_bk_r_03 {parent=dyn_skirt_bk_r_02, loc=<13.337 0 0>, rot=<0.0457 0.0104 -0.0038 -0.9989>}
 dyn_skirt_r_mstr {parent=skirt_base, loc=<3.7974 -1.6408 14.1876>, rot=<0.1199 -0.003 0.9927 0.0107>}
 dyn_skirt_r_01 {parent=dyn_skirt_r_mstr, loc=<0 0 0>, rot=<0 0 0 1>}
 dyn_skirt_r_02 {parent=dyn_skirt_r_01, loc=<13.2632 0 0>, rot=<-0.0001 -0.0093 -0.0001 -1>}
 dyn_skirt_r_03 {parent=dyn_skirt_r_02, loc=<12.8211 0 0>, rot=<-0.0003 -0.0129 -0.0063 -0.9999>}
 dyn_skirt_fr_r_mstr {parent=skirt_base, loc=<3.2066 7.5151 9.6443>, rot=<0.1219 0.4842 0.8664 0.0116>}
 dyn_skirt_fr_r_01 {parent=dyn_skirt_fr_r_mstr, loc=<0 0 0>, rot=<0 0 0 1>}
 dyn_skirt_fr_r_02 {parent=dyn_skirt_fr_r_01, loc=<12.7806 0 0>, rot=<-0.0407 -0.0054 0.0059 -0.9991>}
 dyn_skirt_fr_r_03 {parent=dyn_skirt_fr_r_02, loc=<11.5865 0 0>, rot=<-0.0805 -0.0139 -0.0205 -0.9964>}
 dyn_skirt_l_mstr {parent=skirt_base, loc=<3.7975 -1.6408 -14.1876>, rot=<-0.0107 -0.9927 0.003 -0.1199>}
 dyn_skirt_l_01 {parent=dyn_skirt_l_mstr, loc=<0 0 0>, rot=<0 0 0 1>}
 dyn_skirt_l_02 {parent=dyn_skirt_l_01, loc=<13.2634 0 0>, rot=<0.0001 -0.0093 0.0001 -1>}
 dyn_skirt_l_03 {parent=dyn_skirt_l_02, loc=<12.8211 0 0>, rot=<-0.0063 -0.9999 0.0003 0.0129>}
 dyn_skirt_bk_l_mstr {parent=skirt_base, loc=<3.5231 -9.102 -8.7892>, rot=<0.0177 -0.8385 0.5309 -0.1212>}
 dyn_skirt_bk_l_01 {parent=dyn_skirt_bk_l_mstr, loc=<0 0 0>, rot=<0 0 0 1>}
 dyn_skirt_bk_l_02 {parent=dyn_skirt_bk_l_01, loc=<10.7604 0 -0.0001>, rot=<0.0033 -0.0928 0.0227 -0.9954>}
 dyn_skirt_bk_l_03 {parent=dyn_skirt_bk_l_02, loc=<13.3369 0 0>, rot=<-0.0372 0.0105 0.0038 -0.9992>}
 dyn_skirt_fr_l_mstr {parent=skirt_base, loc=<3.2061 7.5151 -9.6443>, rot=<-0.0142 -0.8767 -0.4652 -0.1216>}
 dyn_skirt_fr_l_01 {parent=dyn_skirt_fr_l_mstr, loc=<0 0 0>, rot=<0 0 0 1>}
 dyn_skirt_fr_l_02 {parent=dyn_skirt_fr_l_01, loc=<12.7801 -0.0001 0>, rot=<0.0382 -0.0056 -0.0057 -0.9992>}
 dyn_skirt_fr_l_03 {parent=dyn_skirt_fr_l_02, loc=<11.5865 0 0>, rot=<0.0092 -0.0114 0.022 -0.9997>}
 dyn_flap_fr_l_mstr {parent=skirt_base, loc=<-2.566 9.4882 -5.3507>, rot=<-0.0495 -0.7062 -0.7058 -0.0273>}
 dyn_flap_fr_l_01 {parent=dyn_flap_fr_l_mstr, loc=<0 0 0>, rot=<0 0 0 1>}
 dyn_flap_fr_l_02 {parent=dyn_flap_fr_l_01, loc=<12.7947 0 0>, rot=<0.0004 -0.0227 0.0079 -0.9997>}
 dyn_flap_fr_l_03 {parent=dyn_flap_fr_l_02, loc=<12.8613 0 0>, rot=<0.0001 -0.0138 0.0078 -0.9999>}
 dyn_flap_fr_r_mstr {parent=skirt_base, loc=<-2.243 9.4766 4.9117>, rot=<-0.0495 -0.7062 -0.7058 -0.0273>}
 dyn_flap_fr_r_01 {parent=dyn_flap_fr_r_mstr, loc=<0 0 0>, rot=<0 0 0 1>}
 dyn_flap_fr_r_02 {parent=dyn_flap_fr_r_01, loc=<12.7947 0 0>, rot=<0.0004 -0.0227 0.0079 -0.9997>}
 dyn_flap_fr_r_03 {parent=dyn_flap_fr_r_02, loc=<12.8613 0 0>, rot=<0.0001 -0.0138 0.0078 -0.9999>}
 dyn_flap_r_mstr {parent=skirt_base, loc=<1.7117 6.917 9.4203>, rot=<0.0904 0.5661 0.8194 -0.0056>}
 dyn_flap_r_01 {parent=dyn_flap_r_mstr, loc=<0 0 0>, rot=<0 0 0 1>}
 dyn_flap_r_02 {parent=dyn_flap_r_01, loc=<12.9366 0 0>, rot=<0.0212 0.0112 0.015 -0.9996>}
 dyn_flap_r_03 {parent=dyn_flap_r_02, loc=<12.5257 0 0>, rot=<-0.0072 -0.0546 0.0108 -0.9984>}
 dyn_flap_l_mstr {parent=skirt_base, loc=<1.7117 6.917 -9.4203>, rot=<0.0056 -0.8194 -0.5661 -0.0904>}
 dyn_flap_l_01 {parent=dyn_flap_l_mstr, loc=<0 0 0>, rot=<0 0 0 1>}
 dyn_flap_l_02 {parent=dyn_flap_l_01, loc=<12.9366 0 0>, rot=<-0.0212 0.0112 -0.015 -0.9996>}
 dyn_flap_l_03 {parent=dyn_flap_l_02, loc=<12.5257 0 0>, rot=<0.0072 -0.0546 -0.0108 -0.9984>}
 dyn_sash_base_mstr {parent=skirt_base, loc=<3.1828 10.0316 -0.0297>, rot=<-0.0138 -0.7092 -0.7045 -0.0236>}
 dyn_sash_base_01 {parent=dyn_sash_base_mstr, loc=<0 0 0>, rot=<0 0 0 1>}
 dyn_sash_r_mstr {parent=dyn_sash_base_01, loc=<1.2942 1.27 0.1802>, rot=<0 0 -0.1799 -0.9837>}
 dyn_sash_r_01 {parent=dyn_sash_r_mstr, loc=<0 0 0>, rot=<0 0 0 1>}
 dyn_sash_r_02 {parent=dyn_sash_r_01, loc=<7.1644 0 0>, rot=<0.0083 0.0446 0.1317 -0.9902>}
 dyn_sash_r_03 {parent=dyn_sash_r_02, loc=<12.5486 0 0>, rot=<0.0014 -0.0179 0.015 -0.9997>}
 dyn_sash_l_mstr {parent=dyn_sash_base_01, loc=<1.3283 -1.1717 0.1953>, rot=<0.0061 0.0011 0.1662 -0.9861>}
 dyn_sash_l_01 {parent=dyn_sash_l_mstr, loc=<0 0 0>, rot=<0 0 0 1>}
 dyn_sash_l_02 {parent=dyn_sash_l_01, loc=<7.1644 0 0>, rot=<-0.0083 0.0446 -0.1317 -0.9902>}
 dyn_sash_l_03 {parent=dyn_sash_l_02, loc=<12.5487 0 0>, rot=<-0.0014 -0.0179 -0.015 -0.9997>}
 dyn_sash_mid_01 {parent=dyn_sash_base_01, loc=<7.3648 0.577 -0.5548>, rot=<-0.0016 0.0297 -0.007 -0.9995>}
 dyn_sash_mid_02 {parent=dyn_sash_mid_01, loc=<10.7323 -0.317 0.0657>, rot=<0.0014 -0.0179 0.015 -0.9997>}
 dyn_flap_mid_01 {parent=dyn_sash_base_01, loc=<7.3648 0.577 -0.5548>, rot=<-0.0016 0.0297 -0.007 -0.9995>}
 dyn_flap_mid_02 {parent=dyn_flap_mid_01, loc=<10.7323 -0.317 0.0657>, rot=<0.0014 -0.0179 0.015 -0.9997>}
 dyn_flap_mid_r_01 {parent=dyn_sash_base_01, loc=<7.3648 0.577 -0.5548>, rot=<-0.0016 0.0297 -0.007 -0.9995>}
 dyn_flap_mid_r_02 {parent=dyn_flap_mid_r_01, loc=<10.7323 -0.317 0.0657>, rot=<0.0014 -0.0179 0.015 -0.9997>}
 dyn_flap_mid_l_01 {parent=dyn_sash_base_01, loc=<7.3648 0.577 -0.5548>, rot=<-0.0016 0.0297 -0.007 -0.9995>}
 dyn_flap_mid_l_02 {parent=dyn_flap_mid_l_01, loc=<10.7323 -0.317 0.0657>, rot=<0.0014 -0.0179 0.015 -0.9997>}
]
variables [
 crouching {type=double, default=0}
 l_leg_up {type=double, default=0}
 r_leg_lock {type=bool, default=False}
 r_leg_up {type=double, default=0}
 r_leg_up_extreme {type=double, default=0}
]
