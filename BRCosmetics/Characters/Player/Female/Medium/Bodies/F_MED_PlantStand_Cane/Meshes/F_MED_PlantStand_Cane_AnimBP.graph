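# Animation Blueprint for the F_MED_PlantStand_Cane female player body cosmetic, adding secondary motion to its skirt, flaps, sash and pants.
# It copies the attached parent mesh's pose into a Control Rig driven by crouch and jump values, then additively offsets skirt bones, partly by crouch state.
# RigidBody physics and AnimDynamics follow, with weighted constraints tying sash and middle flap bones to neighboring bones.
# A post Control Rig then feeds further sash AnimDynamics.
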AnimBlueprint /BRCosmetics/Characters/Player/Female/Medium/Bodies/F_MED_PlantStand_Cane/Meshes/F_MED_PlantStand_Cane_AnimBP
target_skeleton=&/F_MED_PlantStand_Cane_Skeleton
graphs [
 AnimGraph {nodes [
   Anim Dynamics - Bone: dyn_sash_r_03!(ComponentPose, Alpha=Get crouching() -> (crouching:double), GravityOverride:Vector="0,0,-980", ExternalForce="0,0,-980") -> (Pose) {
    angularDampingOverride=0.9
    angularSpringConstant=200
    bGravityOverrideInSimSpace=true
    bOverrideAngularDamping=true
    bUseGravityOverride=true
    bUseSphericalLimits=true
    gravityScale=0.3
    simulationSpace=BoneRelative
    alphaBoolBlend {blendOption=Linear}
    boundBone {boneName=dyn_sash_r_03}
    componentAppliedLinearAccClamp=<2500 2500 2500>
    componentLinearAccScale=<0.7 0.7 0.7>
    componentLinearVelScale=<0.55 0.55 0.55>
    physicsBodyDefinitions [{
      sphereCollisionRadius=0
      boundBone {boneName=dyn_sash_r_03}
      boxExtents=<12 12 12>
      constraintSetup {angularLimitsMax=<0 0 8>, angularLimitsMin=<0 -10 -8>, angularTarget=<0 0 0>}
      localJointOffset=<0.8 0 0>
     }]
    planarLimits [
     {drivingBone {boneName=dyn_flap_fr_r_02}, planeTransform {location=<0 0 0>, rotation=<rot 5 0 0>}}
     {drivingBone {boneName=None}, planeTransform {location=<0 0 0>, rotation=<rot 0 0 0>}}
    ]
    relativeSpaceBone {boneName=root}
    sphericalLimits [
     {limitRadius=6, drivingBone {boneName=dyn_flap_mid_02}, sphereLocalOffset=<2 0 -2>}
     {limitRadius=0, drivingBone {boneName=None}, sphereLocalOffset=<10 0 -4>}
    ]
   }
   n1 = Anim Dynamics - Bone: dyn_sash_l_02!(ComponentPose, Alpha=1, GravityOverride:Vector="0,0,-980", ExternalForce="0,0,-980") -> (Pose) {
    angularDampingOverride=0.9
    angularSpringConstant=200
    bGravityOverrideInSimSpace=true
    bOverrideAngularDamping=true
    bUseGravityOverride=true
    bUseSphericalLimits=true
    gravityScale=0.3
    simulationSpace=BoneRelative
    alphaBoolBlend {blendOption=Linear}
    boundBone {boneName=dyn_sash_l_02}
    componentAppliedLinearAccClamp=<2500 2500 2500>
    componentLinearAccScale=<0.7 0.7 0.7>
    componentLinearVelScale=<0.55 0.55 0.55>
    physicsBodyDefinitions [{
      sphereCollisionRadius=0
      boundBone {boneName=dyn_sash_l_02}
      boxExtents=<12 12 12>
      constraintSetup {angularLimitsMax=<0 0 7.5>, angularLimitsMin=<0 -10 -7.5>, angularTarget=<0 0 0>}
      localJointOffset=<0.8 0 0>
     }]
    planarLimits [
     {drivingBone {boneName=dyn_flap_fr_l_01}, planeTransform {location=<0 0 0>, rotation=<rot 0 0 0>}}
     {drivingBone {boneName=None}, planeTransform {location=<0 0 0>, rotation=<rot 0 0 0>}}
    ]
    relativeSpaceBone {boneName=root}
    sphericalLimits [{limitRadius=5, drivingBone {boneName=dyn_flap_mid_01}, sphereLocalOffset=<5 0 -4>}]
   }
   Anim Dynamics - Bone: dyn_sash_l_03!(ComponentPose=@n1, Alpha=1, GravityOverride:Vector="0,0,-980", ExternalForce="0,0,-980") -> (Pose) {
    angularDampingOverride=0.9
    angularSpringConstant=200
    bGravityOverrideInSimSpace=true
    bOverrideAngularDamping=true
    bUseGravityOverride=true
    bUseSphericalLimits=true
    gravityScale=0.3
    simulationSpace=BoneRelative
    alphaBoolBlend {blendOption=Linear}
    boundBone {boneName=dyn_sash_l_03}
    componentAppliedLinearAccClamp=<2500 2500 2500>
    componentLinearAccScale=<0.7 0.7 0.7>
    componentLinearVelScale=<0.55 0.55 0.55>
    physicsBodyDefinitions [{
      sphereCollisionRadius=0
      boundBone {boneName=dyn_sash_l_03}
      boxExtents=<12 12 12>
      constraintSetup {angularLimitsMax=<0 0 8>, angularLimitsMin=<0 -10 -8>, angularTarget=<0 0 0>}
      localJointOffset=<0.8 0 0>
     }]
    planarLimits [
     {drivingBone {boneName=dyn_flap_fr_l_02}, planeTransform {location=<0 0 0>, rotation=<rot 0 0 0>}}
     {drivingBone {boneName=None}, planeTransform {location=<0 0 0>, rotation=<rot 0 0 0>}}
    ]
    relativeSpaceBone {boneName=root}
    sphericalLimits [{limitRadius=5, drivingBone {boneName=dyn_flap_mid_02}, sphereLocalOffset=<5 0 -5>}]
   }
   n2 = Anim Dynamics - Bone: dyn_sash_r_02!(ComponentPose, Alpha=1, GravityOverride:Vector="0,0,-980", ExternalForce="0,0,-980") -> (Pose) {
    angularDampingOverride=0.9
    angularSpringConstant=200
    bGravityOverrideInSimSpace=true
    bOverrideAngularDamping=true
    bUseGravityOverride=true
    bUseSphericalLimits=true
    gravityScale=0.3
    simulationSpace=BoneRelative
    alphaBoolBlend {blendOption=Linear}
    boundBone {boneName=dyn_sash_r_02}
    componentAppliedLinearAccClamp=<2500 2500 2500>
    componentLinearAccScale=<0.7 0.7 0.7>
    componentLinearVelScale=<0.55 0.55 0.55>
    physicsBodyDefinitions [{
      sphereCollisionRadius=0
      boundBone {boneName=dyn_sash_r_02}
      boxExtents=<12 12 12>
      constraintSetup {angularLimitsMax=<0 0 7.5>, angularLimitsMin=<0 -10 -7.5>, angularTarget=<0 0 0>}
      localJointOffset=<0.8 0 0>
     }]
    planarLimits [
     {drivingBone {boneName=dyn_flap_fr_r_01}, planeTransform {location=<0 0 0>, rotation=<rot 0 0 0>}}
     {drivingBone {boneName=None}, planeTransform {location=<0 3 0>, rotation=<rot 0 0 0>}}
    ]
    relativeSpaceBone {boneName=root}
    sphericalLimits [
     {limitRadius=6, drivingBone {boneName=dyn_flap_mid_01}, sphereLocalOffset=<5 0 -4>}
     {limitRadius=5, drivingBone {boneName=dyn_flap_mid_02}, sphereLocalOffset=<0 0 -3>}
    ]
   }
   Anim Dynamics - Bone: dyn_sash_r_03(ComponentPose=@n2, Alpha=1, GravityOverride="0,0,-980", ExternalForce="0,0,-980") {
    angularDampingOverride=0.9
    angularSpringConstant=200
    bGravityOverrideInSimSpace=true
    bOverrideAngularDamping=true
    bUseGravityOverride=true
    bUseSphericalLimits=true
    gravityScale=0.3
    simulationSpace=BoneRelative
    alphaBoolBlend {blendOption=Linear}
    boundBone {boneName=dyn_sash_r_03}
    componentAppliedLinearAccClamp=<2500 2500 2500>
    componentLinearAccScale=<0.7 0.7 0.7>
    componentLinearVelScale=<0.55 0.55 0.55>
    physicsBodyDefinitions [{
      sphereCollisionRadius=0
      boundBone {boneName=dyn_sash_r_03}
      boxExtents=<12 12 12>
      constraintSetup {angularLimitsMax=<0 0 8>, angularLimitsMin=<0 -10 -8>, angularTarget=<0 0 0>}
      localJointOffset=<0.8 0 0>
     }]
    planarLimits [
     {drivingBone {boneName=dyn_flap_fr_r_02}, planeTransform {location=<0 0 0>, rotation=<rot 5 0 0>}}
     {drivingBone {boneName=None}, planeTransform {location=<0 0 0>, rotation=<rot 0 0 0>}}
    ]
    relativeSpaceBone {boneName=root}
    sphericalLimits [
     {limitRadius=6, drivingBone {boneName=None}, sphereLocalOffset=<2 0 -2>}
     {limitRadius=0, drivingBone {boneName=None}, sphereLocalOffset=<10 0 -4>}
    ]
   }
   Blend Poses by bool(BlendPose_0=Component To Local(), BlendPose_1=Component To Local(), BlendTime_0=0.3, bActiveValue=Get bIsMoving2D()) {blendType=Linear}
   Local To Component(LocalPose=Blend Poses by bool(bActiveValue=Get bIsCrouching()) {blendType=Linear})
   n3 = Control Rig!(
    Source=Copy Pose From Mesh() {bCopyCurves=true, bUseAttachedParent=true}
    Alpha
    crouching:double=Get crouching()
    crouch_moving:double=Get crouch_moving() -> (crouch_moving:double)
    jumping:double=Get jumping() -> (jumping:double)
   ) -> (Pose) {
    alphaBoolBlend {blendOption=Linear}
    controlRigAssetReference {blueprintRigClass=&/F_MED_PlantStand_Cane_CtrlRig.F_MED_PlantStand_Cane_CtrlRig_C}
   }
   Local To Component(LocalPose=@n3)
   n4 = Control Rig(
    Source=Copy Pose From Mesh() {bCopyCurves=true, bUseAttachedParent=true}
    crouching=Get crouching()
    crouch_moving=Get crouch_moving()
    jumping=Get jumping()
   ) {
    alphaBoolBlend {blendOption=Linear}
    controlRigAssetReference {blueprintRigClass=&F_MED_PlantStand_Cane_CtrlRig_C}
   }
   Local To Component(LocalPose=@n4)
   Local To Component(LocalPose=Use cached pose 'Main_Pose'())
   Local To Component(LocalPose=Blend Poses by bool(BlendPose_0=Component To Local(), BlendPose_1=Use cached pose 'Main_Pose'(), BlendTime_0=0.05, BlendTime_1=0.3) {blendType=Linear})
   n5 = Control Rig(
    Source=Copy Pose From Mesh() {bCopyCurves=true, bUseAttachedParent=true}
    crouching=Get crouching()
    crouch_moving=Get crouch_moving()
    jumping=Get jumping()
   ) {
    alphaBoolBlend {blendOption=Linear}
    controlRigAssetReference {blueprintRigClass=&F_MED_PlantStand_Cane_CtrlRig_C}
   }
   n6 = Get crouch_moving()
   n7 = "Transform (Modify) Bone - Bone: dyn_skirt_l_mstr"(ComponentPose=Local To Component(LocalPose=@n5), Alpha=@n6, Translation="0,0,0", Rotation="10,-10,0", Scale="1,1,1") {
    rotationMode=BMM_Additive
    rotationSpace=BCS_BoneSpace
    translationMode=BMM_Additive
    translationSpace=BCS_BoneSpace
    alphaBoolBlend {blendOption=Linear}
    alphaScaleBiasClamp {bInterpResult=true, interpSpeedIncreasing=30}
    boneToModify {boneName=dyn_skirt_l_mstr}
    rotation {pitch=10, yaw=-10}
   }
   n8 = "Transform (Modify) Bone - Bone: dyn_skirt_r_mstr"(ComponentPose=@n7, Alpha=@n6, Translation="0,0,0", Rotation="10,10,0", Scale="1,1,1") {
    rotationMode=BMM_Additive
    rotationSpace=BCS_BoneSpace
    translationMode=BMM_Additive
    translationSpace=BCS_BoneSpace
    alphaBoolBlend {blendOption=Linear}
    alphaScaleBiasClamp {bInterpResult=true, interpSpeedIncreasing=30}
    boneToModify {boneName=dyn_skirt_r_mstr}
    rotation {pitch=10, yaw=10}
   }
   n9 = "Transform (Modify) Bone - Bone: dyn_skirt_bk_l_mstr"(ComponentPose=@n8, Alpha=@n6, Translation="0,0,0", Rotation="25,0,0", Scale="1,1,1") {
    rotationMode=BMM_Additive
    rotationSpace=BCS_BoneSpace
    translationMode=BMM_Additive
    translationSpace=BCS_BoneSpace
    alphaBoolBlend {blendOption=Linear}
    alphaScaleBiasClamp {bInterpResult=true, interpSpeedIncreasing=30}
    boneToModify {boneName=dyn_skirt_bk_l_mstr}
    rotation {pitch=25}
   }
   n10 = "Transform (Modify) Bone - Bone: dyn_skirt_bk_r_mstr"(ComponentPose=@n9, Alpha=@n6, Translation="0,0,0", Rotation="20,0,0", Scale="1,1,1") {
    rotationMode=BMM_Additive
    rotationSpace=BCS_BoneSpace
    translationMode=BMM_Additive
    translationSpace=BCS_BoneSpace
    alphaBoolBlend {blendOption=Linear}
    alphaScaleBiasClamp {bInterpResult=true, interpSpeedIncreasing=30}
    boneToModify {boneName=dyn_skirt_bk_r_mstr}
    rotation {pitch=20}
   }
   n11 = "Transform (Modify) Bone - Bone: dyn_skirt_bk_r_mstr"(ComponentPose=@n10, Alpha=@n6, Translation="0,0,0", Rotation="0,0,0", Scale="1,1,1") {
    rotationMode=BMM_Additive
    rotationSpace=BCS_BoneSpace
    translationMode=BMM_Additive
    translationSpace=BCS_BoneSpace
    alphaBoolBlend {blendOption=Linear}
    alphaScaleBiasClamp {bInterpResult=true, interpSpeedIncreasing=30}
    boneToModify {boneName=dyn_skirt_bk_r_mstr}
   }
   n12 = "Transform (Modify) Bone - Bone: dyn_skirt_bk_04"(ComponentPose=@n11, Alpha=@n6, Translation="0,0,0", Rotation="-60,0,0", Scale="1,1,1") {
    rotationMode=BMM_Additive
    rotationSpace=BCS_BoneSpace
    translationMode=BMM_Additive
    translationSpace=BCS_BoneSpace
    alphaBoolBlend {blendOption=Linear}
    alphaScaleBiasClamp {bInterpResult=true, interpSpeedIncreasing=30}
    boneToModify {boneName=dyn_skirt_bk_04}
    rotation {pitch=-60}
   }
   n13 = Get Crouch Still() -> (Crouch Still:double)
   n14 = "Transform (Modify) Bone - Bone: dyn_skirt_l_mstr"(ComponentPose=@n12, Alpha=@n13, Translation="0,0,0", Rotation="20,-20,0", Scale="1,1,1") {
    rotationMode=BMM_Additive
    rotationSpace=BCS_BoneSpace
    translationMode=BMM_Additive
    translationSpace=BCS_BoneSpace
    alphaBoolBlend {blendOption=Linear}
    alphaScaleBiasClamp {bInterpResult=true, interpSpeedIncreasing=30}
    boneToModify {boneName=dyn_skirt_l_mstr}
    rotation {pitch=20, yaw=-20}
   }
   n15 = "Transform (Modify) Bone - Bone: dyn_skirt_r_mstr"(ComponentPose=@n14, Alpha=@n13, Translation="0,0,0", Rotation="10,20,0", Scale="1,1,1") {
    rotationMode=BMM_Additive
    rotationSpace=BCS_BoneSpace
    translationMode=BMM_Additive
    translationSpace=BCS_BoneSpace
    alphaBoolBlend {blendOption=Linear}
    alphaScaleBiasClamp {bInterpResult=true, interpSpeedIncreasing=30}
    boneToModify {boneName=dyn_skirt_r_mstr}
    rotation {pitch=10, yaw=20}
   }
   n16 = "Transform (Modify) Bone - Bone: dyn_skirt_bk_l_mstr"(ComponentPose=@n15, Alpha=@n13, Translation="0,0,0", Rotation="25,0,0", Scale="1,1,1") {
    rotationMode=BMM_Additive
    rotationSpace=BCS_BoneSpace
    translationMode=BMM_Additive
    translationSpace=BCS_BoneSpace
    alphaBoolBlend {blendOption=Linear}
    alphaScaleBiasClamp {bInterpResult=true, interpSpeedIncreasing=30}
    boneToModify {boneName=dyn_skirt_bk_l_mstr}
    rotation {pitch=25}
   }
   n17 = "Transform (Modify) Bone - Bone: dyn_skirt_bk_r_mstr"(ComponentPose=@n16, Alpha=@n13, Translation="0,0,0", Rotation="65,0,20", Scale="1,1,1") {
    rotationMode=BMM_Additive
    rotationSpace=BCS_BoneSpace
    translationMode=BMM_Additive
    translationSpace=BCS_BoneSpace
    alphaBoolBlend {blendOption=Linear}
    alphaScaleBiasClamp {bInterpResult=true, interpSpeedIncreasing=30}
    boneToModify {boneName=dyn_skirt_bk_r_mstr}
    rotation {pitch=65, roll=20}
   }
   n18 = "Transform (Modify) Bone - Bone: dyn_skirt_bk_mstr"(ComponentPose=@n17, Alpha=@n13, Translation="0,0,0", Rotation="10,0,0", Scale="1,1,1") {
    rotationMode=BMM_Additive
    rotationSpace=BCS_BoneSpace
    translationMode=BMM_Additive
    translationSpace=BCS_BoneSpace
    alphaBoolBlend {blendOption=Linear}
    alphaScaleBiasClamp {bInterpResult=true, interpSpeedIncreasing=30}
    boneToModify {boneName=dyn_skirt_bk_mstr}
    rotation {pitch=10}
   }
   n19 = "Transform (Modify) Bone - Bone: dyn_skirt_bk_04"(ComponentPose=@n18, Alpha=@n13, Translation="0,0,0", Rotation="0,0,0", Scale="1,1,1") {
    rotationMode=BMM_Additive
    rotationSpace=BCS_BoneSpace
    translationMode=BMM_Additive
    translationSpace=BCS_BoneSpace
    alphaBoolBlend {blendOption=Linear}
    alphaScaleBiasClamp {bInterpResult=true, interpSpeedIncreasing=30}
    boneToModify {boneName=dyn_skirt_bk_04}
   }
   n20 = "Transform (Modify) Bone - Bone: dyn_skirt_bk_04"(ComponentPose=@n19, Alpha=@n13, Translation="0,0,0", Rotation="-60,0,0", Scale="1,1,1") {
    rotationMode=BMM_Additive
    rotationSpace=BCS_BoneSpace
    translationMode=BMM_Additive
    translationSpace=BCS_BoneSpace
    alphaBoolBlend {blendOption=Linear}
    alphaScaleBiasClamp {bInterpResult=true, interpSpeedIncreasing=30}
    boneToModify {boneName=dyn_skirt_bk_04}
    rotation {pitch=-60}
   }
   n21 = "Transform (Modify) Bone - Bone: dyn_skirt_bk_mstr"!(ComponentPose=@n20, bAlphaBoolEnabled:bool=True, Translation="0,0,0", Rotation="60,0,0", Scale="1,1,1") -> (Pose) {
    alphaInputType=Bool
    rotationMode=BMM_Additive
    rotationSpace=BCS_BoneSpace
    translationMode=BMM_Additive
    translationSpace=BCS_BoneSpace
    alphaBoolBlend {blendOption=Linear}
    boneToModify {boneName=dyn_skirt_bk_mstr}
    rotation {pitch=60}
   }
   n22 = "Transform (Modify) Bone - Bone: dyn_skirt_bk_r_mstr"!(ComponentPose=@n21, bAlphaBoolEnabled:bool=True, Translation="0,0,0", Rotation="60,0,0", Scale="1,1,1") -> (Pose) {
    alphaInputType=Bool
    rotationMode=BMM_Additive
    rotationSpace=BCS_BoneSpace
    translationMode=BMM_Additive
    translationSpace=BCS_BoneSpace
    alphaBoolBlend {blendOption=Linear}
    boneToModify {boneName=dyn_skirt_bk_r_mstr}
    rotation {pitch=60}
   }
   n23 = "Transform (Modify) Bone - Bone: dyn_skirt_bk_l_mstr"!(ComponentPose=@n22, bAlphaBoolEnabled:bool=True, Translation="0,0,0", Rotation="60,0,0", Scale="1,1,1") -> (Pose) {
    alphaInputType=Bool
    rotationMode=BMM_Additive
    rotationSpace=BCS_BoneSpace
    translationMode=BMM_Additive
    translationSpace=BCS_BoneSpace
    alphaBoolBlend {blendOption=Linear}
    boneToModify {boneName=dyn_skirt_bk_l_mstr}
    rotation {pitch=60}
   }
   n24 = "Transform (Modify) Bone - Bone: dyn_skirt_fr_r_mstr"!(ComponentPose=@n23, bAlphaBoolEnabled:bool=True, Translation="0,0,0", Rotation="70,0,0", Scale="1,1,1") -> (Pose) {
    alphaInputType=Bool
    rotationMode=BMM_Additive
    rotationSpace=BCS_BoneSpace
    translationMode=BMM_Additive
    translationSpace=BCS_BoneSpace
    alphaBoolBlend {blendOption=Linear}
    boneToModify {boneName=dyn_skirt_fr_r_mstr}
    rotation {pitch=70}
   }
   n25 = "Transform (Modify) Bone - Bone: dyn_skirt_r_mstr"!(ComponentPose=@n24, bAlphaBoolEnabled:bool=True, Translation="0,0,0", Rotation="40,0,0", Scale="1,1,1") -> (Pose) {
    alphaInputType=Bool
    rotationMode=BMM_Additive
    rotationSpace=BCS_BoneSpace
    translationMode=BMM_Additive
    translationSpace=BCS_BoneSpace
    alphaBoolBlend {blendOption=Linear}
    boneToModify {boneName=dyn_skirt_r_mstr}
    rotation {pitch=40}
   }
   n26 = "Transform (Modify) Bone - Bone: dyn_skirt_l_mstr"!(ComponentPose=@n25, bAlphaBoolEnabled:bool=True, Translation="0,0,0", Rotation="10,10,0", Scale="1,1,1") -> (Pose) {
    alphaInputType=Bool
    rotationMode=BMM_Additive
    rotationSpace=BCS_BoneSpace
    translationMode=BMM_Additive
    translationSpace=BCS_BoneSpace
    alphaBoolBlend {blendOption=Linear}
    boneToModify {boneName=dyn_skirt_l_mstr}
    rotation {pitch=10, yaw=10}
   }
   n27 = RigidBody!(ComponentPose=@n26, Alpha, OverrideWorldGravity:Vector="0,0,-980", ExternalForce) -> (Pose) {
    bClampLinearTranslationLimitToRefPose=true
    bOverrideWorldGravity=true
    overridePhysicsAsset=&/F_MED_PlantStand_Cane_Physics
    simulationSpace=BaseBoneSpace
    alphaBoolBlend {blendOption=Linear}
    baseBoneRef {boneName=pelvis}
    componentAppliedLinearAccClamp=<2500 2500 2500>
    componentLinearAccScale=<0.35 0.35 0.8>
    componentLinearVelScale=<0.5 0.5 0.7>
    overrideWorldGravity {z=-980}
    simSpaceSettings {
     maxAngularAcceleration=100
     maxAngularVelocity=3
     maxLinearAcceleration=700
     maxLinearVelocity=500
     velocityScaleZ=0
     worldAlpha=1
     externalLinearDragV=<0.4 0.4 0.4>
    }
   }
   n28 = Anim Dynamics - Bone: dyn_pants_l!(ComponentPose=@n27, Alpha=1, GravityOverride:Vector="0,0,-980", ExternalForce="0,0,0") -> (Pose) {
    angularDampingOverride=0.99
    angularSpringConstant=100
    bLinearSpring=true
    bOverrideAngularDamping=true
    bOverrideLinearDamping=true
    bUseGravityOverride=true
    linearDampingOverride=0.92
    linearSpringConstant=30
    simulationSpace=BoneRelative
    alphaBoolBlend {blendOption=Linear}
    boundBone {boneName=dyn_pants_l}
    componentAppliedLinearAccClamp=<1500 1500 1500>
    componentLinearAccScale=<0.8 0.8 0.9>
    componentLinearVelScale=<0.8 0.8 0.9>
    physicsBodyDefinitions [{
      boundBone {boneName=dyn_pants_l}
      boxExtents=<3.5 3.5 3.5>
      constraintSetup {angularLimitsMax=<7 7 7>, angularLimitsMin=<-7 -7 -7>, angularTarget=<0 0 0>, linearAxesMax=<5 5 1>, linearAxesMin=<-5 -5 -1>}
      localJointOffset=<1.3 0 0>
     }]
    relativeSpaceBone {boneName=root}
   }
   n29 = Anim Dynamics - Bone: dyn_pants_r!(ComponentPose=@n28, Alpha=1, GravityOverride:Vector="0,0,-980", ExternalForce="0,0,0") -> (Pose) {
    angularDampingOverride=0.99
    angularSpringConstant=100
    bLinearSpring=true
    bOverrideAngularDamping=true
    bOverrideLinearDamping=true
    bUseGravityOverride=true
    linearDampingOverride=0.92
    linearSpringConstant=30
    simulationSpace=BoneRelative
    alphaBoolBlend {blendOption=Linear}
    boundBone {boneName=dyn_pants_r}
    componentAppliedLinearAccClamp=<1500 1500 1500>
    componentLinearAccScale=<0.8 0.8 0.9>
    componentLinearVelScale=<0.8 0.8 0.9>
    physicsBodyDefinitions [{
      boundBone {boneName=dyn_pants_r}
      boxExtents=<3.5 3.5 3.5>
      constraintSetup {angularLimitsMax=<7 7 7>, angularLimitsMin=<-7 -7 -7>, angularTarget=<0 0 0>, linearAxesMax=<5 5 1>, linearAxesMin=<-5 -5 -1>}
      localJointOffset=<1.3 0 0>
     }]
    relativeSpaceBone {boneName=root}
   }
   n30 = Anim Dynamics - Bone: dyn_flap_fr_l_01!(ComponentPose=@n29, Alpha=1, GravityOverride:Vector="0,0,-980", ExternalForce="0,0,-980") -> (Pose) {
    angularDampingOverride=0.99
    angularSpringConstant=200
    bGravityOverrideInSimSpace=true
    bOverrideAngularDamping=true
    bUseGravityOverride=true
    gravityScale=0.3
    simulationSpace=BoneRelative
    alphaBoolBlend {blendOption=Linear}
    boundBone {boneName=dyn_flap_fr_l_01}
    componentAppliedLinearAccClamp=<2500 2500 2500>
    componentLinearAccScale=<0.5 0.4 0.55>
    componentLinearVelScale=<0.55 0.55 0.55>
    physicsBodyDefinitions [{
      sphereCollisionRadius=0
      boundBone {boneName=dyn_flap_fr_l_01}
      boxExtents=<50 50 50>
      constraintSetup {angularLimitsMax=<0 0 0>, angularLimitsMin=<0 -30 0>, angularTarget=<0 0 0>}
      localJointOffset=<9 0 0>
     }]
    relativeSpaceBone {boneName=root}
   }
   n31 = Anim Dynamics - Bone: dyn_flap_fr_l_02!(ComponentPose=@n30, Alpha=1, GravityOverride:Vector="0,0,-980", ExternalForce="0,0,-980") -> (Pose) {
    angularDampingOverride=0.99
    angularSpringConstant=200
    bGravityOverrideInSimSpace=true
    bOverrideAngularDamping=true
    bUseGravityOverride=true
    gravityScale=0.3
    simulationSpace=BoneRelative
    alphaBoolBlend {blendOption=Linear}
    boundBone {boneName=dyn_flap_fr_l_02}
    componentAppliedLinearAccClamp=<2500 2500 2500>
    componentLinearAccScale=<0.5 0.4 0.55>
    componentLinearVelScale=<0.55 0.55 0.55>
    physicsBodyDefinitions [{
      sphereCollisionRadius=0
      boundBone {boneName=dyn_flap_fr_l_02}
      boxExtents=<50 50 50>
      constraintSetup {angularLimitsMax=<0 0 0>, angularLimitsMin=<0 -30 0>, angularTarget=<0 0 0>}
      localJointOffset=<9 0 0>
     }]
    relativeSpaceBone {boneName=root}
   }
   n32 = Anim Dynamics - Bone: dyn_flap_fr_r_01!(ComponentPose=@n31, Alpha=1, GravityOverride:Vector="0,0,-980", ExternalForce="0,0,-980") -> (Pose) {
    angularDampingOverride=0.99
    angularSpringConstant=200
    bGravityOverrideInSimSpace=true
    bOverrideAngularDamping=true
    bUseGravityOverride=true
    gravityScale=0.3
    simulationSpace=BoneRelative
    alphaBoolBlend {blendOption=Linear}
    boundBone {boneName=dyn_flap_fr_r_01}
    componentAppliedLinearAccClamp=<2500 2500 2500>
    componentLinearAccScale=<0.5 0.4 0.55>
    componentLinearVelScale=<0.55 0.55 0.55>
    physicsBodyDefinitions [{
      sphereCollisionRadius=0
      boundBone {boneName=dyn_flap_fr_r_01}
      boxExtents=<50 50 50>
      constraintSetup {angularLimitsMax=<0 0 0>, angularLimitsMin=<0 -30 0>, angularTarget=<0 0 0>}
      localJointOffset=<9 0 0>
     }]
    relativeSpaceBone {boneName=root}
   }
   n33 = Anim Dynamics - Bone: dyn_flap_fr_r_02!(ComponentPose=@n32, Alpha=1, GravityOverride:Vector="0,0,-980", ExternalForce="0,0,-980") -> (Pose) {
    angularDampingOverride=0.99
    angularSpringConstant=200
    bGravityOverrideInSimSpace=true
    bOverrideAngularDamping=true
    bUseGravityOverride=true
    gravityScale=0.3
    simulationSpace=BoneRelative
    alphaBoolBlend {blendOption=Linear}
    boundBone {boneName=dyn_flap_fr_r_02}
    componentAppliedLinearAccClamp=<2500 2500 2500>
    componentLinearAccScale=<0.5 0.4 0.55>
    componentLinearVelScale=<0.55 0.55 0.55>
    physicsBodyDefinitions [{
      sphereCollisionRadius=0
      boundBone {boneName=dyn_flap_fr_r_02}
      boxExtents=<50 50 50>
      constraintSetup {angularLimitsMax=<0 0 0>, angularLimitsMin=<0 -30 0>, angularTarget=<0 0 0>}
      localJointOffset=<9 0 0>
     }]
    relativeSpaceBone {boneName=root}
   }
   n34 = "Constraint - dyn_sash_l_01 (PPPPPPPPPPPPP)"!(
    ComponentPose=@n33
    Alpha=1
    ConstraintWeights_0=0.5129
    ConstraintWeights_1:float=0.0103
    ConstraintWeights_2:float=0.001
    ConstraintWeights_3:float=0
    ConstraintWeights_4:float=0.0001
    ConstraintWeights_5:float=0.0014
    ConstraintWeights_6:float=0.0815
    ConstraintWeights_7:float=0.0002
    ConstraintWeights_8:float=0.0037
    ConstraintWeights_9:float=0.0029
    ConstraintWeights_10:float=0.085
    ConstraintWeights_11:float=0.0033
    ConstraintWeights_12:float=0.2978
   ) -> (Pose) {
    alphaBoolBlend {blendOption=Linear}
    boneToModify {boneName=dyn_sash_l_01}
    constraintSetup [
     {targetBone {boneName=pelvis}}
     {targetBone {boneName=thigh_twist_01_l}}
     {targetBone {boneName=thigh_twist_01_r}}
     {targetBone {boneName=thigh_r}}
     {targetBone {boneName=thigh_l}}
     {targetBone {boneName=dyn_flap_l_02}}
     {targetBone {boneName=dyn_flap_l_01}}
     {targetBone {boneName=dyn_flap_r_02}}
     {targetBone {boneName=dyn_flap_r_01}}
     {targetBone {boneName=dyn_flap_fr_r_02}}
     {targetBone {boneName=dyn_flap_fr_r_01}}
     {targetBone {boneName=dyn_flap_fr_l_02}}
     {targetBone {boneName=dyn_flap_fr_l_01}}
    ]
    constraintWeights [1,1,1,1,1,1,1,1,1,1,1,1,1]
   }
   n35 = Get crouching()
   n36 = "Transform (Modify) Bone - Bone: dyn_sash_base_mstr"(ComponentPose=@n34, Alpha=@n35, Translation="-2.5,0,0", Rotation="0,0,0", Scale="1,1,1") {
    rotationMode=BMM_Additive
    rotationSpace=BCS_BoneSpace
    translationMode=BMM_Additive
    translationSpace=BCS_BoneSpace
    alphaBoolBlend {blendOption=Linear}
    boneToModify {boneName=dyn_sash_base_mstr}
    translation {x=-2.5}
   }
   n37 = "Transform (Modify) Bone - Bone: dyn_sash_l_01"(ComponentPose=@n36, Alpha=@n35, Translation="5,0,0", Rotation="0,0,0", Scale="1,1,1") {
    rotationMode=BMM_Additive
    rotationSpace=BCS_BoneSpace
    translationMode=BMM_Additive
    translationSpace=BCS_BoneSpace
    alphaBoolBlend {blendOption=Linear}
    boneToModify {boneName=dyn_sash_l_01}
    translation {x=5}
   }
   n38 = "Constraint - dyn_sash_l_02 (PP)"!(ComponentPose=@n37, Alpha=1, ConstraintWeights_0=1, ConstraintWeights_1:float=0) -> (Pose) {
    alphaBoolBlend {blendOption=Linear}
    boneToModify {boneName=dyn_sash_l_02}
    constraintSetup [{targetBone {boneName=dyn_flap_fr_l_01}}, {targetBone {boneName=pelvis}}]
    constraintWeights [1,1]
   }
   n39 = "Transform (Modify) Bone - Bone: dyn_sash_l_02"(ComponentPose=@n38, Alpha=@n35, Translation="2,0,0", Rotation="0,0,0", Scale="1,1,1") {
    rotationMode=BMM_Additive
    rotationSpace=BCS_BoneSpace
    translationMode=BMM_Additive
    translationSpace=BCS_BoneSpace
    alphaBoolBlend {blendOption=Linear}
    boneToModify {boneName=dyn_sash_l_02}
    translation {x=2}
   }
   n40 = "Constraint - dyn_sash_l_03 (P)"(ComponentPose=@n39, Alpha=1, ConstraintWeights_0=1) {
    alphaBoolBlend {blendOption=Linear}
    boneToModify {boneName=dyn_sash_l_03}
    constraintSetup [{targetBone {boneName=dyn_flap_fr_l_02}}]
    constraintWeights [1]
   }
   n41 = "Constraint - dyn_sash_r_01 (PP)"!(ComponentPose=@n40, Alpha=1, ConstraintWeights_0=0.5, ConstraintWeights_1:float=0.5) -> (Pose) {
    alphaBoolBlend {blendOption=Linear}
    boneToModify {boneName=dyn_sash_r_01}
    constraintSetup [{targetBone {boneName=dyn_flap_fr_r_01}}, {targetBone {boneName=dyn_sash_base_01}}]
    constraintWeights [1,1]
   }
   n42 = "Constraint - dyn_sash_r_02 (P)"(ComponentPose=@n41, Alpha=1, ConstraintWeights_0=1) {
    alphaBoolBlend {blendOption=Linear}
    boneToModify {boneName=dyn_sash_r_02}
    constraintSetup [{targetBone {boneName=dyn_flap_fr_r_01}}]
    constraintWeights [1]
   }
   n43 = "Transform (Modify) Bone - Bone: dyn_sash_r_02"(ComponentPose=@n42, Alpha=1, Translation="0,0,0.5", Rotation="0,0,0", Scale="1,1,1") {
    rotationMode=BMM_Additive
    rotationSpace=BCS_BoneSpace
    translationMode=BMM_Additive
    translationSpace=BCS_BoneSpace
    alphaBoolBlend {blendOption=Linear}
    boneToModify {boneName=dyn_sash_r_02}
    translation {z=0.5}
   }
   n44 = "Transform (Modify) Bone - Bone: dyn_sash_r_02"(ComponentPose=@n43, Alpha=Get crouching(), Translation="0,0,0.5", Rotation="0,0,0", Scale="1,1,1") {
    rotationMode=BMM_Additive
    rotationSpace=BCS_BoneSpace
    translationMode=BMM_Additive
    translationSpace=BCS_BoneSpace
    alphaBoolBlend {blendOption=Linear}
    boneToModify {boneName=dyn_sash_r_02}
    translation {z=0.5}
   }
   n45 = "Constraint - dyn_sash_r_03 (P)"(ComponentPose=@n44, Alpha=1, ConstraintWeights_0=1) {
    alphaBoolBlend {blendOption=Linear}
    boneToModify {boneName=dyn_sash_r_03}
    constraintSetup [{targetBone {boneName=dyn_flap_fr_r_02}}]
    constraintWeights [1]
   }
   n46 = "Constraint - dyn_flap_mid_01 (PPPPPPPPPPPPP)"!(
    ComponentPose=@n45
    Alpha=1
    ConstraintWeights_0=0.3675
    ConstraintWeights_1:float=0.0023
    ConstraintWeights_2:float=0.0027
    ConstraintWeights_3:float=0
    ConstraintWeights_4:float=0
    ConstraintWeights_5:float=0.0001
    ConstraintWeights_6:float=0.0234
    ConstraintWeights_7:float=0.0007
    ConstraintWeights_8:float=0.0114
    ConstraintWeights_9:float=0.0049
    ConstraintWeights_10:float=0.3331
    ConstraintWeights_11:float=0.0039
    ConstraintWeights_12:float=0.2499
   ) -> (Pose) {
    alphaBoolBlend {blendOption=Linear}
    boneToModify {boneName=dyn_flap_mid_01}
    constraintSetup [
     {targetBone {boneName=pelvis}}
     {targetBone {boneName=thigh_twist_01_l}}
     {targetBone {boneName=thigh_twist_01_r}}
     {targetBone {boneName=thigh_r}}
     {targetBone {boneName=thigh_l}}
     {targetBone {boneName=dyn_flap_l_02}}
     {targetBone {boneName=dyn_flap_l_01}}
     {targetBone {boneName=dyn_flap_r_02}}
     {targetBone {boneName=dyn_flap_r_01}}
     {targetBone {boneName=dyn_flap_fr_r_02}}
     {targetBone {boneName=dyn_flap_fr_r_01}}
     {targetBone {boneName=dyn_flap_fr_l_02}}
     {targetBone {boneName=dyn_flap_fr_l_01}}
    ]
    constraintWeights [1,1,1,1,1,1,1,1,1,1,1,1,1]
   }
   n47 = "Constraint - dyn_flap_mid_02 (PPPPPPPPPPPP)"!(
    ComponentPose=@n46
    Alpha=1
    ConstraintWeights_0=0.0103
    ConstraintWeights_1:float=0.0001
    ConstraintWeights_2:float=0.0001
    ConstraintWeights_3:float=0
    ConstraintWeights_4:float=0
    ConstraintWeights_5:float=0.001
    ConstraintWeights_6:float=0
    ConstraintWeights_7:float=0.0004
    ConstraintWeights_8:float=0.2438
    ConstraintWeights_9:float=0.2352
    ConstraintWeights_10:float=0.1202
    ConstraintWeights_11:float=0.3889
   ) -> (Pose) {
    alphaBoolBlend {blendOption=Linear}
    boneToModify {boneName=dyn_flap_mid_02}
    constraintSetup [
     {targetBone {boneName=pelvis}}
     {targetBone {boneName=thigh_twist_01_l}}
     {targetBone {boneName=thigh_twist_01_r}}
     {targetBone {boneName=thigh_r}}
     {targetBone {boneName=thigh_l}}
     {targetBone {boneName=dyn_flap_l_01}}
     {targetBone {boneName=dyn_flap_r_02}}
     {targetBone {boneName=dyn_flap_r_01}}
     {targetBone {boneName=dyn_flap_fr_r_02}}
     {targetBone {boneName=dyn_flap_fr_r_01}}
     {targetBone {boneName=dyn_flap_fr_l_02}}
     {targetBone {boneName=dyn_flap_fr_l_01}}
    ]
    constraintWeights [1,1,1,1,1,1,1,1,1,1,1,1]
   }
   n48 = Control Rig!(Source=Component To Local(ComponentPose=@n47), Alpha, crouching=Get crouching(), r_leg_lock:bool=Get r_skirt_lock() -> (r_skirt_lock:bool)) -> (Pose) {
    alphaBoolBlend {blendOption=Linear}
    controlRigAssetReference {blueprintRigClass=&/F_MED_PlantStand_Cane_Post_CtrlRig.F_MED_PlantStand_Cane_Post_CtrlRig_C}
   }
   n49 = Anim Dynamics - Bone: dyn_sash_r_01!(ComponentPose=Local To Component(LocalPose=@n48), bAlphaBoolEnabled:bool=True, GravityOverride:Vector="0,0,-980", ExternalForce="0,0,0") -> (Pose) {
    alphaInputType=Bool
    angularDampingOverride=0.9
    angularSpringConstant=200
    bGravityOverrideInSimSpace=true
    bOverrideAngularDamping=true
    bUseGravityOverride=true
    bUsePlanarLimit=false
    gravityScale=0.3
    simulationSpace=BoneRelative
    alphaBoolBlend {blendInTime=0.1, blendOption=Linear, blendOutTime=0.1}
    boundBone {boneName=dyn_sash_r_01}
    componentAppliedLinearAccClamp=<2500 2500 2500>
    componentLinearAccScale=<0.7 0.7 0.7>
    componentLinearVelScale=<0.55 0.55 0.55>
    physicsBodyDefinitions [{
      sphereCollisionRadius=0
      boundBone {boneName=dyn_sash_r_01}
      boxExtents=<12 12 12>
      constraintSetup {angularLimitsMax=<0 0 2>, angularLimitsMin=<0 -45 -2>, angularTarget=<0 0 0>}
      localJointOffset=<0.8 0 0>
     }]
    relativeSpaceBone {boneName=root}
   }
   n50 = Anim Dynamics - Bone: dyn_sash_r_02!(ComponentPose=@n49, bAlphaBoolEnabled:bool=True, GravityOverride="0,0,-980", ExternalForce="0,0,0") -> (Pose) {
    alphaInputType=Bool
    angularDampingOverride=0.9
    angularSpringConstant=200
    bGravityOverrideInSimSpace=true
    bOverrideAngularDamping=true
    bUseGravityOverride=true
    bUsePlanarLimit=false
    gravityScale=0.3
    simulationSpace=BoneRelative
    alphaBoolBlend {blendInTime=0.1, blendOption=Linear, blendOutTime=0.1}
    boundBone {boneName=dyn_sash_r_02}
    componentAppliedLinearAccClamp=<2500 2500 2500>
    componentLinearAccScale=<0.7 0.7 0.7>
    componentLinearVelScale=<0.55 0.55 0.55>
    physicsBodyDefinitions [{
      sphereCollisionRadius=0
      boundBone {boneName=dyn_sash_r_02}
      boxExtents=<12 12 12>
      constraintSetup {angularLimitsMax=<0 0 7.5>, angularLimitsMin=<0 -10 -7.5>, angularTarget=<0 0 0>}
      localJointOffset=<0.8 0 0>
     }]
    relativeSpaceBone {boneName=root}
    sphericalLimits [
     {limitRadius=6, drivingBone {boneName=dyn_flap_mid_01}, sphereLocalOffset=<5 0 -4>}
     {limitRadius=5, drivingBone {boneName=dyn_flap_mid_02}, sphereLocalOffset=<0 0 -3>}
    ]
   }
   n51 = Anim Dynamics - Bone: dyn_sash_r_03!(ComponentPose=@n50, bAlphaBoolEnabled:bool=True, GravityOverride="0,0,-980", ExternalForce="0,0,0") -> (Pose) {
    alphaInputType=Bool
    angularDampingOverride=0.9
    angularSpringConstant=200
    bGravityOverrideInSimSpace=true
    bOverrideAngularDamping=true
    bUseGravityOverride=true
    bUsePlanarLimit=false
    gravityScale=0.3
    simulationSpace=BoneRelative
    alphaBoolBlend {blendInTime=0.1, blendOption=Linear, blendOutTime=0.1}
    boundBone {boneName=dyn_sash_r_03}
    componentAppliedLinearAccClamp=<2500 2500 2500>
    componentLinearAccScale=<0.7 0.7 0.7>
    componentLinearVelScale=<0.55 0.55 0.55>
    physicsBodyDefinitions [{
      sphereCollisionRadius=0
      boundBone {boneName=dyn_sash_r_03}
      boxExtents=<12 12 12>
      constraintSetup {angularLimitsMax=<0 0 8>, angularLimitsMin=<0 -10 -8>, angularTarget=<0 0 0>}
      localJointOffset=<0.8 0 0>
     }]
    relativeSpaceBone {boneName=root}
    sphericalLimits [
     {limitRadius=6, drivingBone {boneName=None}, sphereLocalOffset=<2 0 -2>}
     {limitRadius=0, drivingBone {boneName=None}, sphereLocalOffset=<10 0 -4>}
    ]
   }
   n52 = Anim Dynamics - Bone: dyn_sash_l_01!(ComponentPose=@n51, bAlphaBoolEnabled:bool=True, GravityOverride:Vector="0,0,-980", ExternalForce="0,0,0") -> (Pose) {
    alphaInputType=Bool
    angularDampingOverride=0.9
    angularSpringConstant=200
    bGravityOverrideInSimSpace=true
    bOverrideAngularDamping=true
    bUseGravityOverride=true
    bUsePlanarLimit=false
    gravityScale=0.3
    simulationSpace=BoneRelative
    alphaBoolBlend {blendInTime=0.1, blendOption=Linear, blendOutTime=0.1}
    boundBone {boneName=dyn_sash_l_01}
    componentAppliedLinearAccClamp=<2500 2500 2500>
    componentLinearAccScale=<0.7 0.7 0.7>
    componentLinearVelScale=<0.55 0.55 0.55>
    physicsBodyDefinitions [{
      sphereCollisionRadius=0
      boundBone {boneName=dyn_sash_l_01}
      boxExtents=<12 12 12>
      constraintSetup {angularLimitsMax=<0 0 2>, angularLimitsMin=<0 -45 -2>, angularTarget=<0 0 0>}
      localJointOffset=<0.8 0 0>
     }]
    relativeSpaceBone {boneName=root}
   }
   n53 = Anim Dynamics - Bone: dyn_sash_l_02!(ComponentPose=@n52, bAlphaBoolEnabled:bool=True, GravityOverride="0,0,-980", ExternalForce="0,0,0") -> (Pose) {
    alphaInputType=Bool
    angularDampingOverride=0.9
    angularSpringConstant=200
    bGravityOverrideInSimSpace=true
    bOverrideAngularDamping=true
    bUseGravityOverride=true
    bUsePlanarLimit=false
    gravityScale=0.3
    simulationSpace=BoneRelative
    alphaBoolBlend {blendInTime=0.1, blendOption=Linear, blendOutTime=0.1}
    boundBone {boneName=dyn_sash_l_02}
    componentAppliedLinearAccClamp=<2500 2500 2500>
    componentLinearAccScale=<0.7 0.7 0.7>
    componentLinearVelScale=<0.55 0.55 0.55>
    physicsBodyDefinitions [{
      sphereCollisionRadius=0
      boundBone {boneName=dyn_sash_l_02}
      boxExtents=<12 12 12>
      constraintSetup {angularLimitsMax=<0 0 7.5>, angularLimitsMin=<0 -10 -7.5>, angularTarget=<0 0 0>}
      localJointOffset=<0.8 0 0>
     }]
    relativeSpaceBone {boneName=root}
    sphericalLimits [{limitRadius=5, drivingBone {boneName=dyn_flap_mid_01}, sphereLocalOffset=<5 0 -4>}]
   }
   n54 = Anim Dynamics - Bone: dyn_sash_l_03!(ComponentPose=@n53, bAlphaBoolEnabled:bool=True, GravityOverride="0,0,-980", ExternalForce="0,0,0") -> (Pose) {
    alphaInputType=Bool
    angularDampingOverride=0.9
    angularSpringConstant=200
    bGravityOverrideInSimSpace=true
    bOverrideAngularDamping=true
    bUseGravityOverride=true
    bUsePlanarLimit=false
    gravityScale=0.3
    simulationSpace=BoneRelative
    alphaBoolBlend {blendInTime=0.1, blendOption=Linear, blendOutTime=0.1}
    boundBone {boneName=dyn_sash_l_03}
    componentAppliedLinearAccClamp=<2500 2500 2500>
    componentLinearAccScale=<0.7 0.7 0.7>
    componentLinearVelScale=<0.55 0.55 0.55>
    physicsBodyDefinitions [{
      sphereCollisionRadius=0
      boundBone {boneName=dyn_sash_l_03}
      boxExtents=<12 12 12>
      constraintSetup {angularLimitsMax=<0 0 8>, angularLimitsMin=<0 -10 -8>, angularTarget=<0 0 0>}
      localJointOffset=<0.8 0 0>
     }]
    relativeSpaceBone {boneName=root}
    sphericalLimits [{limitRadius=5, drivingBone {boneName=dyn_flap_mid_02}, sphereLocalOffset=<5 0 -5>}]
   }
   n55 = Anim Dynamics - Bone: dyn_sash_r_01(ComponentPose=@n54, bAlphaBoolEnabled=True, GravityOverride="0,0,-980", ExternalForce="0,0,-980") {
    alphaInputType=Bool
    angularDampingOverride=0.9
    angularSpringConstant=200
    bGravityOverrideInSimSpace=true
    bOverrideAngularDamping=true
    bUseGravityOverride=true
    bUsePlanarLimit=false
    gravityScale=0.3
    simulationSpace=BoneRelative
    alphaBoolBlend {blendInTime=0.1, blendOption=Linear, blendOutTime=0.1}
    boundBone {boneName=dyn_sash_r_01}
    componentAppliedLinearAccClamp=<2500 2500 2500>
    componentLinearAccScale=<0.7 0.7 0.7>
    componentLinearVelScale=<0.55 0.55 0.55>
    physicsBodyDefinitions [{
      sphereCollisionRadius=0
      boundBone {boneName=dyn_sash_r_01}
      boxExtents=<12 12 12>
      constraintSetup {angularLimitsMax=<0 0 2>, angularLimitsMin=<0 -45 -2>, angularTarget=<0 0 0>}
      localJointOffset=<0.8 0 0>
     }]
    planarLimits [{drivingBone {boneName=dyn_flap_fr_r_01}, planeTransform {location=<0 0 0>, rotation=<rot 0 0 0>}}]
    relativeSpaceBone {boneName=root}
   }
   n56 = Anim Dynamics - Bone: dyn_sash_r_02(ComponentPose=@n55, bAlphaBoolEnabled=True, GravityOverride="0,0,-980", ExternalForce="0,0,-980") {
    alphaInputType=Bool
    angularDampingOverride=0.9
    angularSpringConstant=200
    bGravityOverrideInSimSpace=true
    bOverrideAngularDamping=true
    bUseGravityOverride=true
    gravityScale=0.3
    simulationSpace=BoneRelative
    alphaBoolBlend {blendInTime=0.1, blendOption=Linear, blendOutTime=0.1}
    boundBone {boneName=dyn_sash_r_02}
    componentAppliedLinearAccClamp=<2500 2500 2500>
    componentLinearAccScale=<0.7 0.7 0.7>
    componentLinearVelScale=<0.55 0.55 0.55>
    physicsBodyDefinitions [{
      sphereCollisionRadius=0
      boundBone {boneName=dyn_sash_r_02}
      boxExtents=<12 12 12>
      constraintSetup {angularLimitsMax=<0 0 7.5>, angularLimitsMin=<0 -10 -7.5>, angularTarget=<0 0 0>}
      localJointOffset=<0.8 0 0>
     }]
    planarLimits [
     {drivingBone {boneName=dyn_flap_fr_r_01}, planeTransform {location=<0 0 0>, rotation=<rot 0 0 0>}}
     {drivingBone {boneName=dyn_flap_mid_r_01}, planeTransform {location=<0 3 0>, rotation=<rot 0 0 0>}}
    ]
    relativeSpaceBone {boneName=root}
    sphericalLimits [
     {limitRadius=6, drivingBone {boneName=dyn_flap_mid_01}, sphereLocalOffset=<5 0 -4>}
     {limitRadius=5, drivingBone {boneName=dyn_flap_mid_02}, sphereLocalOffset=<0 0 -3>}
    ]
   }
   n57 = Anim Dynamics - Bone: dyn_sash_r_03(ComponentPose=@n56, bAlphaBoolEnabled=True, GravityOverride="0,0,-980", ExternalForce="0,0,-980") {
    alphaInputType=Bool
    angularDampingOverride=0.9
    angularSpringConstant=200
    bGravityOverrideInSimSpace=true
    bOverrideAngularDamping=true
    bUseGravityOverride=true
    gravityScale=0.3
    simulationSpace=BoneRelative
    alphaBoolBlend {blendInTime=0.1, blendOption=Linear, blendOutTime=0.1}
    boundBone {boneName=dyn_sash_r_03}
    componentAppliedLinearAccClamp=<2500 2500 2500>
    componentLinearAccScale=<0.7 0.7 0.7>
    componentLinearVelScale=<0.55 0.55 0.55>
    physicsBodyDefinitions [{
      sphereCollisionRadius=0
      boundBone {boneName=dyn_sash_r_03}
      boxExtents=<12 12 12>
      constraintSetup {angularLimitsMax=<0 0 8>, angularLimitsMin=<0 -10 -8>, angularTarget=<0 0 0>}
      localJointOffset=<0.8 0 0>
     }]
    planarLimits [
     {drivingBone {boneName=dyn_flap_fr_r_02}, planeTransform {location=<0 0 0.5>, rotation=<rot 0 0 5>}}
     {drivingBone {boneName=dyn_flap_mid_r_02}, planeTransform {location=<0 0 1>, rotation=<rot 0 0 0>}}
    ]
    relativeSpaceBone {boneName=root}
    sphericalLimits [
     {limitRadius=6, drivingBone {boneName=None}, sphereLocalOffset=<2 0 -2>}
     {limitRadius=0, drivingBone {boneName=None}, sphereLocalOffset=<10 0 -4>}
    ]
   }
   n58 = Anim Dynamics - Bone: dyn_sash_l_01(ComponentPose=@n57, bAlphaBoolEnabled=True, GravityOverride="0,0,-980", ExternalForce="0,0,-980") {
    alphaInputType=Bool
    angularDampingOverride=0.9
    angularSpringConstant=200
    bGravityOverrideInSimSpace=true
    bOverrideAngularDamping=true
    bUseGravityOverride=true
    bUsePlanarLimit=false
    gravityScale=0.3
    simulationSpace=BoneRelative
    alphaBoolBlend {blendInTime=0.1, blendOption=Linear, blendOutTime=0.1}
    boundBone {boneName=dyn_sash_l_01}
    componentAppliedLinearAccClamp=<2500 2500 2500>
    componentLinearAccScale=<0.7 0.7 0.7>
    componentLinearVelScale=<0.55 0.55 0.55>
    physicsBodyDefinitions [{
      sphereCollisionRadius=0
      boundBone {boneName=dyn_sash_l_01}
      boxExtents=<12 12 12>
      constraintSetup {angularLimitsMax=<0 0 2>, angularLimitsMin=<0 -45 -2>, angularTarget=<0 0 0>}
      localJointOffset=<0.8 0 0>
     }]
    planarLimits [{drivingBone {boneName=dyn_flap_fr_l_01}, planeTransform {location=<0 0 0>, rotation=<rot 0 0 0>}}]
    relativeSpaceBone {boneName=root}
   }
   n59 = Anim Dynamics - Bone: dyn_sash_l_02(ComponentPose=@n58, bAlphaBoolEnabled=True, GravityOverride="0,0,-980", ExternalForce="0,0,-980") {
    alphaInputType=Bool
    angularDampingOverride=0.9
    angularSpringConstant=200
    bGravityOverrideInSimSpace=true
    bOverrideAngularDamping=true
    bUseGravityOverride=true
    gravityScale=0.3
    simulationSpace=BoneRelative
    alphaBoolBlend {blendInTime=0.1, blendOption=Linear, blendOutTime=0.1}
    boundBone {boneName=dyn_sash_l_02}
    componentAppliedLinearAccClamp=<2500 2500 2500>
    componentLinearAccScale=<0.7 0.7 0.7>
    componentLinearVelScale=<0.55 0.55 0.55>
    physicsBodyDefinitions [{
      sphereCollisionRadius=0
      boundBone {boneName=dyn_sash_l_02}
      boxExtents=<12 12 12>
      constraintSetup {angularLimitsMax=<0 0 7.5>, angularLimitsMin=<0 -10 -7.5>, angularTarget=<0 0 0>}
      localJointOffset=<0.8 0 0>
     }]
    planarLimits [
     {drivingBone {boneName=dyn_flap_fr_l_01}, planeTransform {location=<0 0 0>, rotation=<rot 0 0 0>}}
     {drivingBone {boneName=dyn_flap_mid_l_01}, planeTransform {location=<0 0 0>, rotation=<rot 0 0 0>}}
    ]
    relativeSpaceBone {boneName=root}
    sphericalLimits [{limitRadius=5, drivingBone {boneName=dyn_flap_mid_01}, sphereLocalOffset=<5 0 -4>}]
   }
   n60 = Anim Dynamics - Bone: dyn_sash_l_03(ComponentPose=@n59, bAlphaBoolEnabled=True, GravityOverride="0,0,-980", ExternalForce="0,0,-980") {
    alphaInputType=Bool
    angularDampingOverride=0.9
    angularSpringConstant=200
    bGravityOverrideInSimSpace=true
    bOverrideAngularDamping=true
    bUseGravityOverride=true
    gravityScale=0.3
    simulationSpace=BoneRelative
    alphaBoolBlend {blendInTime=0.1, blendOption=Linear, blendOutTime=0.1}
    boundBone {boneName=dyn_sash_l_03}
    componentAppliedLinearAccClamp=<2500 2500 2500>
    componentLinearAccScale=<0.7 0.7 0.7>
    componentLinearVelScale=<0.55 0.55 0.55>
    physicsBodyDefinitions [{
      sphereCollisionRadius=0
      boundBone {boneName=dyn_sash_l_03}
      boxExtents=<12 12 12>
      constraintSetup {angularLimitsMax=<0 0 8>, angularLimitsMin=<0 -10 -8>, angularTarget=<0 0 0>}
      localJointOffset=<0.8 0 0>
     }]
    planarLimits [
     {drivingBone {boneName=dyn_flap_fr_l_02}, planeTransform {location=<0 0 0>, rotation=<rot 0 0 0>}}
     {drivingBone {boneName=dyn_flap_mid_l_02}, planeTransform {location=<0 0 0>, rotation=<rot 0 0 0>}}
    ]
    relativeSpaceBone {boneName=root}
    sphericalLimits [{limitRadius=5, drivingBone {boneName=dyn_flap_mid_02}, sphereLocalOffset=<5 0 -5>}]
   }
   n61 = Get bIsJumping()
   n62 = "Transform (Modify) Bone - Bone: dyn_skirt_bk_r_mstr"(ComponentPose=@n60, bAlphaBoolEnabled=@n61, Translation="0,0,0", Rotation="20,0,0", Scale="1,1,1") {
    alphaInputType=Bool
    rotationMode=BMM_Additive
    rotationSpace=BCS_BoneSpace
    translationMode=BMM_Additive
    translationSpace=BCS_BoneSpace
    alphaBoolBlend {blendInTime=0.05, blendOption=Linear, blendOutTime=0.3}
    boneToModify {boneName=dyn_skirt_bk_r_mstr}
    rotation {pitch=20}
   }
   n63 = "Transform (Modify) Bone - Bone: dyn_skirt_r_mstr"(ComponentPose=@n62, bAlphaBoolEnabled=@n61, Translation="0,0,0", Rotation="20,0,0", Scale="1,1,1") {
    alphaInputType=Bool
    rotationMode=BMM_Additive
    rotationSpace=BCS_BoneSpace
    translationMode=BMM_Additive
    translationSpace=BCS_BoneSpace
    alphaBoolBlend {blendInTime=0.05, blendOption=Linear, blendOutTime=0.3}
    boneToModify {boneName=dyn_skirt_r_mstr}
    rotation {pitch=20}
   }
   n64 = "Transform (Modify) Bone - Bone: dyn_skirt_fr_l_mstr"!(ComponentPose=@n63, bAlphaBoolEnabled:bool=Get bIsInFrontEnd(), Translation="0,0,0", Rotation="3,0,0", Scale="1,1,1") -> (Pose) {
    alphaInputType=Bool
    rotationMode=BMM_Additive
    rotationSpace=BCS_BoneSpace
    translationMode=BMM_Additive
    translationSpace=BCS_BoneSpace
    alphaBoolBlend {blendInTime=0.1, blendOption=Linear, blendOutTime=0.1}
    boneToModify {boneName=dyn_skirt_fr_l_mstr}
    rotation {pitch=3}
   }
   n65 = "Transform (Modify) Bone - Bone: dyn_skirt_l_mstr"!(ComponentPose=@n64, Alpha:float=1, Translation="0,0,0", Rotation="20,0,0", Scale="1,1,1") -> (Pose) {
    rotationMode=BMM_Additive
    rotationSpace=BCS_BoneSpace
    translationMode=BMM_Additive
    translationSpace=BCS_BoneSpace
    alphaBoolBlend {blendOption=Linear}
    boneToModify {boneName=dyn_skirt_l_mstr}
    rotation {pitch=20}
   }
   n66 = "Transform (Modify) Bone - Bone: dyn_skirt_r_mstr"!(ComponentPose=@n65, Alpha:float=1, Translation="0,0,0", Rotation="15,0,0", Scale="1,1,1") -> (Pose) {
    rotationMode=BMM_Additive
    rotationSpace=BCS_BoneSpace
    translationMode=BMM_Additive
    translationSpace=BCS_BoneSpace
    alphaBoolBlend {blendOption=Linear}
    boneToModify {boneName=dyn_skirt_r_mstr}
    rotation {pitch=15}
   }
   n67 = "Transform (Modify) Bone - Bone: dyn_skirt_fr_r_mstr"!(ComponentPose=@n66, Alpha:float=1, Translation="0,0,0", Rotation="-10,0,0", Scale="1,1,1") -> (Pose) {
    rotationMode=BMM_Additive
    rotationSpace=BCS_BoneSpace
    translationMode=BMM_Additive
    translationSpace=BCS_BoneSpace
    alphaBoolBlend {blendOption=Linear}
    boneToModify {boneName=dyn_skirt_fr_r_mstr}
    rotation {pitch=-10}
   }
   n68 = "Transform (Modify) Bone - Bone: dyn_skirt_bk_l_mstr"!(ComponentPose=@n67, Alpha:float=1, Translation="0,0,0", Rotation="5,0,0", Scale="1,1,1") -> (Pose) {
    rotationMode=BMM_Additive
    rotationSpace=BCS_BoneSpace
    translationMode=BMM_Additive
    translationSpace=BCS_BoneSpace
    alphaBoolBlend {blendOption=Linear}
    boneToModify {boneName=dyn_skirt_bk_l_mstr}
    rotation {pitch=5}
   }
   n69 = "Transform (Modify) Bone - Bone: dyn_skirt_bk_mstr"!(ComponentPose=@n68, Alpha:float=1, Translation="0,0,0", Rotation="5,0,0", Scale="1,1,1") -> (Pose) {
    rotationMode=BMM_Additive
    rotationSpace=BCS_BoneSpace
    translationMode=BMM_Additive
    translationSpace=BCS_BoneSpace
    alphaBoolBlend {blendOption=Linear}
    boneToModify {boneName=dyn_skirt_bk_mstr}
    rotation {pitch=5}
   }
   n70 = "Transform (Modify) Bone - Bone: dyn_skirt_bk_r_mstr"!(ComponentPose=@n69, Alpha:float=1, Translation="0,0,0", Rotation="5,0,0", Scale="1,1,1") -> (Pose) {
    rotationMode=BMM_Additive
    rotationSpace=BCS_BoneSpace
    translationMode=BMM_Additive
    translationSpace=BCS_BoneSpace
    alphaBoolBlend {blendOption=Linear}
    boneToModify {boneName=dyn_skirt_bk_r_mstr}
    rotation {pitch=5}
   }
   n71 = "Transform (Modify) Bone - Bone: dyn_flap_fr_l_mstr"!(ComponentPose=@n70, bAlphaBoolEnabled:bool=True, Translation="0,0,0", Rotation="-4.926,5.071,5.889") -> (Pose) {
    alphaInputType=Bool
    rotationMode=BMM_Additive
    rotationSpace=BCS_BoneSpace
    translationMode=BMM_Additive
    translationSpace=BCS_BoneSpace
    alphaBoolBlend {blendOption=Linear}
    boneToModify {boneName=dyn_flap_fr_l_mstr}
    rotation=<rot -4.926 5.071 5.889>
   }
   n72 = "Transform (Modify) Bone - Bone: dyn_flap_fr_r_mstr"!(ComponentPose=@n71, bAlphaBoolEnabled:bool=True, Translation="0,0,0", Rotation="2.353,-15.621,-9.664") -> (Pose) {
    alphaInputType=Bool
    rotationMode=BMM_Additive
    rotationSpace=BCS_BoneSpace
    translationMode=BMM_Additive
    translationSpace=BCS_BoneSpace
    alphaBoolBlend {blendOption=Linear}
    boneToModify {boneName=dyn_flap_fr_r_mstr}
    rotation=<rot 2.353 -15.621 -9.664>
   }
   n73 = "Transform (Modify) Bone - Bone: dyn_skirt_fr_l_mstr"!(ComponentPose=@n72, bAlphaBoolEnabled=True, Translation="0,0,0", Rotation="29.837,-10.842,-1.991") -> (Pose) {
    alphaInputType=Bool
    rotationMode=BMM_Additive
    rotationSpace=BCS_BoneSpace
    translationMode=BMM_Additive
    translationSpace=BCS_BoneSpace
    alphaBoolBlend {blendOption=Linear}
    boneToModify {boneName=dyn_skirt_fr_l_mstr}
    rotation=<rot 29.837 -10.842 -1.991>
   }
   n74 = "Transform (Modify) Bone - Bone: dyn_skirt_l_mstr"!(ComponentPose=@n73, bAlphaBoolEnabled:bool=True, Translation="0,0,0", Rotation="14.633,0.000,-0.000") -> (Pose) {
    alphaInputType=Bool
    rotationMode=BMM_Additive
    rotationSpace=BCS_BoneSpace
    translationMode=BMM_Additive
    translationSpace=BCS_BoneSpace
    alphaBoolBlend {blendOption=Linear}
    boneToModify {boneName=dyn_skirt_l_mstr}
    rotation=<rot 14.633 0 0>
   }
   Output Pose(Result=Component To Local(ComponentPose=@n74)) {layerGroup=None}
   n75 = Anim Dynamics - Bone: dyn_skirt_l_01!(
    ComponentPose=Reroute Node(InputPin:Component Space Pose Link Structure) -> (OutputPin:Component Space Pose Link Structure)
    Alpha=1
    GravityOverride:Vector="0,0,-980"
    ExternalForce="0,0,0"
   ) -> (Pose) {
    angularDampingOverride=1
    angularSpringConstant=50
    bAngularSpring=true
    bOverrideAngularDamping=true
    bUseGravityOverride=true
    gravityScale=0.3
    simulationSpace=BoneRelative
    alphaBoolBlend {blendOption=Linear}
    boundBone {boneName=dyn_skirt_l_01}
    componentAppliedLinearAccClamp=<2500 2500 1500>
    componentLinearAccScale=<0.4 0.4 1>
    componentLinearVelScale=<0.55 0.55 0.6>
    physicsBodyDefinitions [{
      collisionType=CustomSphere
      sphereCollisionRadius=2
      boundBone {boneName=dyn_skirt_l_01}
      boxExtents=<20 20 20>
      constraintSetup {angularLimitsMax=<0 0 10>, angularLimitsMin=<0 -50 -10>, angularTarget=<0 0 0>}
      localJointOffset=<9 0 0>
     }]
    relativeSpaceBone {boneName=root}
   }
   n76 = Anim Dynamics - Bone: dyn_skirt_l_02!(ComponentPose=@n75, Alpha=1, GravityOverride:Vector="0,0,-980", ExternalForce="0,0,0") -> (Pose) {
    angularDampingOverride=1
    angularSpringConstant=50
    bAngularSpring=true
    bOverrideAngularDamping=true
    bUseGravityOverride=true
    gravityScale=0.3
    simulationSpace=BoneRelative
    alphaBoolBlend {blendOption=Linear}
    boundBone {boneName=dyn_skirt_l_02}
    componentAppliedLinearAccClamp=<2500 2500 1500>
    componentLinearAccScale=<0.4 0.4 1>
    componentLinearVelScale=<0.55 0.55 0.6>
    physicsBodyDefinitions [{
      collisionType=CustomSphere
      sphereCollisionRadius=2
      boundBone {boneName=dyn_skirt_l_02}
      boxExtents=<20 20 20>
      constraintSetup {angularLimitsMax=<0 0 10>, angularLimitsMin=<0 -50 -10>, angularTarget=<0 0 0>}
      localJointOffset=<9 0 0>
     }]
    relativeSpaceBone {boneName=root}
   }
   n77 = Anim Dynamics - Bone: dyn_skirt_l_03!(ComponentPose=@n76, Alpha=1, GravityOverride:Vector="0,0,-980", ExternalForce="0,0,0") -> (Pose) {
    angularDampingOverride=1
    angularSpringConstant=50
    bAngularSpring=true
    bOverrideAngularDamping=true
    bUseGravityOverride=true
    gravityScale=0.3
    simulationSpace=BoneRelative
    alphaBoolBlend {blendOption=Linear}
    boundBone {boneName=dyn_skirt_l_03}
    componentAppliedLinearAccClamp=<2500 2500 1500>
    componentLinearAccScale=<0.4 0.4 1>
    componentLinearVelScale=<0.55 0.55 0.6>
    physicsBodyDefinitions [{
      collisionType=CustomSphere
      sphereCollisionRadius=2
      boundBone {boneName=dyn_skirt_l_03}
      boxExtents=<20 20 20>
      constraintSetup {angularLimitsMax=<0 0 10>, angularLimitsMin=<0 -50 -10>, angularTarget=<0 0 0>}
      localJointOffset=<-9 0 0>
     }]
    relativeSpaceBone {boneName=root}
   }
   n78 = Anim Dynamics - Bone: dyn_skirt_fr_l_01!(ComponentPose=@n77, Alpha=1, GravityOverride:Vector="0,0,-980", ExternalForce="0,0,0") -> (Pose) {
    angularDampingOverride=1
    angularSpringConstant=50
    bAngularSpring=true
    bOverrideAngularDamping=true
    bUseGravityOverride=true
    gravityScale=0.3
    simulationSpace=BoneRelative
    alphaBoolBlend {blendOption=Linear}
    boundBone {boneName=dyn_skirt_fr_l_01}
    componentAppliedLinearAccClamp=<2500 2500 1500>
    componentLinearAccScale=<0.4 0.4 1>
    componentLinearVelScale=<0.55 0.55 0.6>
    physicsBodyDefinitions [{
      collisionType=CustomSphere
      sphereCollisionRadius=2
      boundBone {boneName=dyn_skirt_fr_l_01}
      boxExtents=<20 20 20>
      constraintSetup {angularLimitsMax=<0 0 10>, angularLimitsMin=<0 -50 -10>, angularTarget=<0 0 0>}
      localJointOffset=<9 0 0>
     }]
    relativeSpaceBone {boneName=root}
   }
   n79 = Anim Dynamics - Bone: dyn_skirt_fr_l_02!(ComponentPose=@n78, Alpha=1, GravityOverride:Vector="0,0,-980", ExternalForce="0,0,0") -> (Pose) {
    angularDampingOverride=1
    angularSpringConstant=50
    bAngularSpring=true
    bOverrideAngularDamping=true
    bUseGravityOverride=true
    gravityScale=0.3
    simulationSpace=BoneRelative
    alphaBoolBlend {blendOption=Linear}
    boundBone {boneName=dyn_skirt_fr_l_02}
    componentAppliedLinearAccClamp=<2500 2500 1500>
    componentLinearAccScale=<0.4 0.4 1>
    componentLinearVelScale=<0.55 0.55 0.6>
    physicsBodyDefinitions [{
      collisionType=CustomSphere
      sphereCollisionRadius=2
      boundBone {boneName=dyn_skirt_fr_l_02}
      boxExtents=<20 20 20>
      constraintSetup {angularLimitsMax=<0 0 10>, angularLimitsMin=<0 -50 -10>, angularTarget=<0 0 0>}
      localJointOffset=<9 0 0>
     }]
    relativeSpaceBone {boneName=root}
   }
   n80 = Anim Dynamics - Bone: dyn_skirt_fr_l_03!(ComponentPose=@n79, Alpha=1, GravityOverride:Vector="0,0,-980", ExternalForce="0,0,0") -> (Pose) {
    angularDampingOverride=1
    angularSpringConstant=50
    bAngularSpring=true
    bOverrideAngularDamping=true
    bUseGravityOverride=true
    gravityScale=0.3
    simulationSpace=BoneRelative
    alphaBoolBlend {blendOption=Linear}
    boundBone {boneName=dyn_skirt_fr_l_03}
    componentAppliedLinearAccClamp=<2500 2500 1500>
    componentLinearAccScale=<0.4 0.4 1>
    componentLinearVelScale=<0.55 0.55 0.6>
    physicsBodyDefinitions [{
      collisionType=CustomSphere
      sphereCollisionRadius=2
      boundBone {boneName=dyn_skirt_fr_l_03}
      boxExtents=<20 20 20>
      constraintSetup {angularLimitsMax=<0 0 10>, angularLimitsMin=<0 -50 -10>, angularTarget=<0 0 0>}
      localJointOffset=<9 0 0>
     }]
    relativeSpaceBone {boneName=root}
   }
   n81 = Anim Dynamics - Bone: dyn_skirt_fr_r_01!(ComponentPose=@n80, Alpha=1, GravityOverride:Vector="0,0,-980", ExternalForce="0,0,0") -> (Pose) {
    angularDampingOverride=1
    angularSpringConstant=50
    bAngularSpring=true
    bOverrideAngularDamping=true
    bUseGravityOverride=true
    gravityScale=0.3
    simulationSpace=BoneRelative
    alphaBoolBlend {blendOption=Linear}
    boundBone {boneName=dyn_skirt_fr_r_01}
    componentAppliedLinearAccClamp=<2500 2500 1500>
    componentLinearAccScale=<0.4 0.4 1>
    componentLinearVelScale=<0.55 0.55 0.6>
    physicsBodyDefinitions [{
      collisionType=CustomSphere
      sphereCollisionRadius=2
      boundBone {boneName=dyn_skirt_fr_r_01}
      boxExtents=<20 20 20>
      constraintSetup {angularLimitsMax=<0 0 10>, angularLimitsMin=<0 -50 -10>, angularTarget=<0 0 0>}
      localJointOffset=<9 0 0>
     }]
    relativeSpaceBone {boneName=root}
   }
   n82 = Anim Dynamics - Bone: dyn_skirt_fr_r_02!(ComponentPose=@n81, Alpha=1, GravityOverride:Vector="0,0,-980", ExternalForce="0,0,0") -> (Pose) {
    angularDampingOverride=1
    angularSpringConstant=50
    bAngularSpring=true
    bOverrideAngularDamping=true
    bUseGravityOverride=true
    gravityScale=0.3
    simulationSpace=BoneRelative
    alphaBoolBlend {blendOption=Linear}
    boundBone {boneName=dyn_skirt_fr_r_02}
    componentAppliedLinearAccClamp=<2500 2500 1500>
    componentLinearAccScale=<0.4 0.4 1>
    componentLinearVelScale=<0.55 0.55 0.6>
    physicsBodyDefinitions [{
      collisionType=CustomSphere
      sphereCollisionRadius=2
      boundBone {boneName=dyn_skirt_fr_r_02}
      boxExtents=<20 20 20>
      constraintSetup {angularLimitsMax=<0 0 10>, angularLimitsMin=<0 -50 -10>, angularTarget=<0 0 0>}
      localJointOffset=<9 0 0>
     }]
    relativeSpaceBone {boneName=root}
   }
   n83 = Anim Dynamics - Bone: dyn_skirt_fr_r_03!(ComponentPose=@n82, Alpha=1, GravityOverride:Vector="0,0,-980", ExternalForce="0,0,0") -> (Pose) {
    angularDampingOverride=1
    angularSpringConstant=50
    bAngularSpring=true
    bOverrideAngularDamping=true
    bUseGravityOverride=true
    gravityScale=0.3
    simulationSpace=BoneRelative
    alphaBoolBlend {blendOption=Linear}
    boundBone {boneName=dyn_skirt_fr_r_03}
    componentAppliedLinearAccClamp=<2500 2500 1500>
    componentLinearAccScale=<0.4 0.4 1>
    componentLinearVelScale=<0.55 0.55 0.6>
    physicsBodyDefinitions [{
      collisionType=CustomSphere
      sphereCollisionRadius=2
      boundBone {boneName=dyn_skirt_fr_r_03}
      boxExtents=<20 20 20>
      constraintSetup {angularLimitsMax=<0 0 10>, angularLimitsMin=<0 -50 -10>, angularTarget=<0 0 0>}
      localJointOffset=<9 0 0>
     }]
    relativeSpaceBone {boneName=root}
   }
   n84 = Anim Dynamics - Bone: dyn_skirt_r_01!(ComponentPose=@n83, Alpha=1, GravityOverride:Vector="0,0,-980", ExternalForce="0,0,0") -> (Pose) {
    angularDampingOverride=1
    angularSpringConstant=50
    bAngularSpring=true
    bOverrideAngularDamping=true
    bUseGravityOverride=true
    gravityScale=0.3
    simulationSpace=BoneRelative
    alphaBoolBlend {blendOption=Linear}
    boundBone {boneName=dyn_skirt_r_01}
    componentAppliedLinearAccClamp=<2500 2500 1500>
    componentLinearAccScale=<0.4 0.4 1>
    componentLinearVelScale=<0.55 0.55 0.6>
    physicsBodyDefinitions [{
      collisionType=CustomSphere
      sphereCollisionRadius=2
      boundBone {boneName=dyn_skirt_r_01}
      boxExtents=<20 20 20>
      constraintSetup {angularLimitsMax=<0 0 10>, angularLimitsMin=<0 -50 -10>, angularTarget=<0 0 0>}
      localJointOffset=<9 0 0>
     }]
    relativeSpaceBone {boneName=root}
   }
   n85 = Anim Dynamics - Bone: dyn_skirt_r_02!(ComponentPose=@n84, Alpha=1, GravityOverride:Vector="0,0,-980", ExternalForce="0,0,0") -> (Pose) {
    angularDampingOverride=1
    angularSpringConstant=50
    bAngularSpring=true
    bOverrideAngularDamping=true
    bUseGravityOverride=true
    gravityScale=0.3
    simulationSpace=BoneRelative
    alphaBoolBlend {blendOption=Linear}
    boundBone {boneName=dyn_skirt_r_02}
    componentAppliedLinearAccClamp=<2500 2500 1500>
    componentLinearAccScale=<0.4 0.4 1>
    componentLinearVelScale=<0.55 0.55 0.6>
    physicsBodyDefinitions [{
      collisionType=CustomSphere
      sphereCollisionRadius=2
      boundBone {boneName=dyn_skirt_r_02}
      boxExtents=<20 20 20>
      constraintSetup {angularLimitsMax=<0 0 10>, angularLimitsMin=<0 -50 -10>, angularTarget=<0 0 0>}
      localJointOffset=<9 0 0>
     }]
    relativeSpaceBone {boneName=root}
    sphericalLimits [
     {limitRadius=20, drivingBone {boneName=calf_r}, sphereLocalOffset=<10 0 0>}
     {limitRadius=20, drivingBone {boneName=calf_r}, sphereLocalOffset=<25 0 0>}
     {limitRadius=0, drivingBone {boneName=None}, sphereLocalOffset=<0 0 0>}
     {limitRadius=0, drivingBone {boneName=None}, sphereLocalOffset=<0 0 0>}
    ]
   }
   n86 = Anim Dynamics - Bone: dyn_skirt_r_03!(ComponentPose=@n85, Alpha=1, GravityOverride:Vector="0,0,-980", ExternalForce="0,0,0") -> (Pose) {
    angularDampingOverride=1
    angularSpringConstant=50
    bAngularSpring=true
    bOverrideAngularDamping=true
    bUseGravityOverride=true
    gravityScale=0.3
    simulationSpace=BoneRelative
    alphaBoolBlend {blendOption=Linear}
    boundBone {boneName=dyn_skirt_r_03}
    componentAppliedLinearAccClamp=<2500 2500 1500>
    componentLinearAccScale=<0.4 0.4 1>
    componentLinearVelScale=<0.55 0.55 0.6>
    physicsBodyDefinitions [{
      collisionType=CustomSphere
      sphereCollisionRadius=2
      boundBone {boneName=dyn_skirt_r_03}
      boxExtents=<20 20 20>
      constraintSetup {angularLimitsMax=<0 0 10>, angularLimitsMin=<0 -50 -10>, angularTarget=<0 0 0>}
      localJointOffset=<9 0 0>
     }]
    relativeSpaceBone {boneName=root}
    sphericalLimits [
     {limitRadius=20, drivingBone {boneName=calf_r}, sphereLocalOffset=<10 0 0>}
     {limitRadius=20, drivingBone {boneName=calf_r}, sphereLocalOffset=<25 0 0>}
     {limitRadius=0, drivingBone {boneName=None}, sphereLocalOffset=<0 0 0>}
     {limitRadius=0, drivingBone {boneName=None}, sphereLocalOffset=<0 0 0>}
    ]
   }
   n87 = Anim Dynamics - Bone: dyn_skirt_bk_01!(ComponentPose=@n86, Alpha=1, GravityOverride:Vector="0,0,-980", ExternalForce="0,0,0") -> (Pose) {
    angularDampingOverride=1
    angularSpringConstant=50
    bAngularSpring=true
    bOverrideAngularDamping=true
    bUseGravityOverride=true
    bUsePlanarLimit=false
    bUseSphericalLimits=true
    gravityScale=0.3
    simulationSpace=BoneRelative
    alphaBoolBlend {blendOption=Linear}
    boundBone {boneName=dyn_skirt_bk_01}
    componentAppliedLinearAccClamp=<2500 2500 1500>
    componentLinearAccScale=<0.4 0.4 1>
    componentLinearVelScale=<0.55 0.55 0.6>
    physicsBodyDefinitions [{
      collisionType=CustomSphere
      sphereCollisionRadius=2
      boundBone {boneName=dyn_skirt_bk_01}
      boxExtents=<20 20 20>
      constraintSetup {angularLimitsMax=<0 30 10>, angularLimitsMin=<0 -50 -10>, angularTarget=<0 0 0>}
      localJointOffset=<9 0 0>
     }]
    relativeSpaceBone {boneName=root}
    sphericalLimits [
     {limitRadius=10, drivingBone {boneName=thigh_l}, sphereLocalOffset=<-10 0 0>}
     {limitRadius=10, drivingBone {boneName=thigh_r}, sphereLocalOffset=<10 0 0>}
     {limitRadius=10, drivingBone {boneName=thigh_l}, sphereLocalOffset=<-30 0 0>}
     {limitRadius=10, drivingBone {boneName=thigh_r}, sphereLocalOffset=<30 0 0>}
    ]
   }
   n88 = Anim Dynamics - Bone: dyn_skirt_bk_02!(ComponentPose=@n87, Alpha=1, GravityOverride:Vector="0,0,-980", ExternalForce="0,0,0") -> (Pose) {
    angularDampingOverride=1
    angularSpringConstant=50
    bAngularSpring=true
    bOverrideAngularDamping=true
    bUseGravityOverride=true
    bUsePlanarLimit=false
    bUseSphericalLimits=true
    gravityScale=0.3
    simulationSpace=BoneRelative
    alphaBoolBlend {blendOption=Linear}
    boundBone {boneName=dyn_skirt_bk_02}
    componentAppliedLinearAccClamp=<2500 2500 1500>
    componentLinearAccScale=<0.4 0.4 1>
    componentLinearVelScale=<0.55 0.55 0.6>
    physicsBodyDefinitions [{
      collisionType=CustomSphere
      sphereCollisionRadius=2
      boundBone {boneName=dyn_skirt_bk_02}
      boxExtents=<20 20 20>
      constraintSetup {angularLimitsMax=<0 30 10>, angularLimitsMin=<0 -50 -10>, angularTarget=<0 0 0>}
      localJointOffset=<9 0 0>
     }]
    relativeSpaceBone {boneName=root}
    sphericalLimits [
     {limitRadius=10, drivingBone {boneName=thigh_l}, sphereLocalOffset=<-10 0 0>}
     {limitRadius=10, drivingBone {boneName=thigh_r}, sphereLocalOffset=<10 0 0>}
     {limitRadius=10, drivingBone {boneName=thigh_l}, sphereLocalOffset=<-30 0 0>}
     {limitRadius=10, drivingBone {boneName=thigh_r}, sphereLocalOffset=<30 0 0>}
    ]
   }
   n89 = Anim Dynamics - Bone: dyn_skirt_bk_03!(ComponentPose=@n88, Alpha=1, GravityOverride:Vector="0,0,-980", ExternalForce="0,0,0") -> (Pose) {
    angularDampingOverride=1
    angularSpringConstant=50
    bAngularSpring=true
    bOverrideAngularDamping=true
    bUseGravityOverride=true
    bUsePlanarLimit=false
    bUseSphericalLimits=true
    gravityScale=0.3
    simulationSpace=BoneRelative
    alphaBoolBlend {blendOption=Linear}
    boundBone {boneName=dyn_skirt_bk_03}
    componentAppliedLinearAccClamp=<2500 2500 1500>
    componentLinearAccScale=<0.4 0.4 1>
    componentLinearVelScale=<0.55 0.55 0.6>
    physicsBodyDefinitions [{
      collisionType=CustomSphere
      sphereCollisionRadius=2
      boundBone {boneName=dyn_skirt_bk_03}
      boxExtents=<20 20 20>
      constraintSetup {angularLimitsMax=<0 30 10>, angularLimitsMin=<0 -50 -10>, angularTarget=<0 0 0>}
      localJointOffset=<9 0 0>
     }]
    relativeSpaceBone {boneName=root}
    sphericalLimits [
     {limitRadius=10, drivingBone {boneName=thigh_l}, sphereLocalOffset=<-10 0 0>}
     {limitRadius=10, drivingBone {boneName=thigh_r}, sphereLocalOffset=<10 0 0>}
     {limitRadius=10, drivingBone {boneName=thigh_l}, sphereLocalOffset=<-30 0 0>}
     {limitRadius=10, drivingBone {boneName=thigh_r}, sphereLocalOffset=<30 0 0>}
     {limitRadius=10, drivingBone {boneName=calf_l}, sphereLocalOffset=<-5 0 0>}
     {limitRadius=10, drivingBone {boneName=calf_r}, sphereLocalOffset=<5 0 0>}
    ]
   }
   n90 = Anim Dynamics - Bone: dyn_skirt_bk_04!(ComponentPose=@n89, Alpha=1, GravityOverride:Vector="0,0,-980", ExternalForce="0,0,0") -> (Pose) {
    angularDampingOverride=1
    angularSpringConstant=50
    bAngularSpring=true
    bOverrideAngularDamping=true
    bUseGravityOverride=true
    bUsePlanarLimit=false
    bUseSphericalLimits=true
    gravityScale=0.3
    simulationSpace=BoneRelative
    alphaBoolBlend {blendOption=Linear}
    boundBone {boneName=dyn_skirt_bk_04}
    componentAppliedLinearAccClamp=<2500 2500 1500>
    componentLinearAccScale=<0.4 0.4 1>
    componentLinearVelScale=<0.55 0.55 0.6>
    physicsBodyDefinitions [{
      collisionType=CustomSphere
      sphereCollisionRadius=2
      boundBone {boneName=dyn_skirt_bk_04}
      boxExtents=<20 20 20>
      constraintSetup {angularLimitsMax=<0 30 10>, angularLimitsMin=<0 -50 -10>, angularTarget=<0 0 0>}
      localJointOffset=<9 0 0>
     }]
    relativeSpaceBone {boneName=root}
    sphericalLimits [
     {limitRadius=10, drivingBone {boneName=thigh_l}, sphereLocalOffset=<-30 0 0>}
     {limitRadius=10, drivingBone {boneName=thigh_r}, sphereLocalOffset=<30 0 0>}
     {limitRadius=10, drivingBone {boneName=calf_l}, sphereLocalOffset=<-5 0 0>}
     {limitRadius=10, drivingBone {boneName=calf_r}, sphereLocalOffset=<5 0 0>}
     {limitRadius=15, drivingBone {boneName=calf_l}, sphereLocalOffset=<-25 0 0>}
     {limitRadius=15, drivingBone {boneName=calf_r}, sphereLocalOffset=<25 0 0>}
     {limitRadius=10, drivingBone {boneName=foot_l}, sphereLocalOffset=<0 0 0>}
     {limitRadius=10, drivingBone {boneName=foot_r}, sphereLocalOffset=<0 0 0>}
    ]
   }
   n91 = Anim Dynamics - Bone: dyn_skirt_bk_r_01!(ComponentPose=@n90, Alpha=1, GravityOverride:Vector="0,0,-980", ExternalForce="0,0,0") -> (Pose) {
    angularDampingOverride=1
    angularSpringConstant=50
    bAngularSpring=true
    bOverrideAngularDamping=true
    bUseGravityOverride=true
    bUsePlanarLimit=false
    bUseSphericalLimits=true
    gravityScale=0.3
    simulationSpace=BoneRelative
    alphaBoolBlend {blendOption=Linear}
    boundBone {boneName=dyn_skirt_bk_r_01}
    componentAppliedLinearAccClamp=<2500 2500 1500>
    componentLinearAccScale=<0.4 0.4 1>
    componentLinearVelScale=<0.55 0.55 0.6>
    physicsBodyDefinitions [{
      collisionType=CustomSphere
      sphereCollisionRadius=2
      boundBone {boneName=dyn_skirt_bk_r_01}
      boxExtents=<20 20 20>
      constraintSetup {angularLimitsMax=<0 30 10>, angularLimitsMin=<0 -50 -10>, angularTarget=<0 0 0>}
      localJointOffset=<9 0 0>
     }]
    relativeSpaceBone {boneName=root}
    sphericalLimits [
     {limitRadius=10, drivingBone {boneName=thigh_r}, sphereLocalOffset=<10 0 0>}
     {limitRadius=10, drivingBone {boneName=thigh_r}, sphereLocalOffset=<30 0 0>}
    ]
   }
   n92 = Anim Dynamics - Bone: dyn_skirt_bk_r_02!(ComponentPose=@n91, Alpha=1, GravityOverride:Vector="0,0,-980", ExternalForce="0,0,0") -> (Pose) {
    angularDampingOverride=1
    angularSpringConstant=50
    bAngularSpring=true
    bOverrideAngularDamping=true
    bUseGravityOverride=true
    bUsePlanarLimit=false
    bUseSphericalLimits=true
    gravityScale=0.3
    simulationSpace=BoneRelative
    alphaBoolBlend {blendOption=Linear}
    boundBone {boneName=dyn_skirt_bk_r_02}
    componentAppliedLinearAccClamp=<2500 2500 1500>
    componentLinearAccScale=<0.4 0.4 1>
    componentLinearVelScale=<0.55 0.55 0.6>
    physicsBodyDefinitions [{
      collisionType=CustomSphere
      sphereCollisionRadius=2
      boundBone {boneName=dyn_skirt_bk_r_02}
      boxExtents=<20 20 20>
      constraintSetup {angularLimitsMax=<0 30 10>, angularLimitsMin=<0 -50 -10>, angularTarget=<0 0 0>}
      localJointOffset=<9 0 0>
     }]
    relativeSpaceBone {boneName=root}
    sphericalLimits [
     {limitRadius=12, drivingBone {boneName=thigh_r}, sphereLocalOffset=<10 0 0>}
     {limitRadius=12, drivingBone {boneName=thigh_r}, sphereLocalOffset=<30 0 0>}
     {limitRadius=12, drivingBone {boneName=calf_r}, sphereLocalOffset=<5 0 0>}
     {limitRadius=20, drivingBone {boneName=foot_r}, sphereLocalOffset=<0 0 0>}
     {limitRadius=15, drivingBone {boneName=calf_r}, sphereLocalOffset=<25 0 0>}
    ]
   }
   n93 = Anim Dynamics - Bone: dyn_skirt_bk_r_03!(ComponentPose=@n92, Alpha=1, GravityOverride:Vector="0,0,-980", ExternalForce="0,0,0") -> (Pose) {
    angularDampingOverride=1
    angularSpringConstant=50
    bAngularSpring=true
    bOverrideAngularDamping=true
    bUseGravityOverride=true
    bUsePlanarLimit=false
    bUseSphericalLimits=true
    gravityScale=0.3
    simulationSpace=BoneRelative
    alphaBoolBlend {blendOption=Linear}
    boundBone {boneName=dyn_skirt_bk_r_03}
    componentAppliedLinearAccClamp=<2500 2500 1500>
    componentLinearAccScale=<0.4 0.4 1>
    componentLinearVelScale=<0.55 0.55 0.6>
    physicsBodyDefinitions [{
      collisionType=CustomSphere
      sphereCollisionRadius=2
      boundBone {boneName=dyn_skirt_bk_r_03}
      boxExtents=<20 20 20>
      constraintSetup {angularLimitsMax=<0 30 10>, angularLimitsMin=<0 -50 -10>, angularTarget=<0 0 0>}
      localJointOffset=<9 0 0>
     }]
    relativeSpaceBone {boneName=root}
    sphericalLimits [
     {limitRadius=12, drivingBone {boneName=thigh_r}, sphereLocalOffset=<10 0 0>}
     {limitRadius=12, drivingBone {boneName=thigh_r}, sphereLocalOffset=<30 0 0>}
     {limitRadius=12, drivingBone {boneName=calf_r}, sphereLocalOffset=<5 0 0>}
     {limitRadius=15, drivingBone {boneName=calf_r}, sphereLocalOffset=<25 0 0>}
     {limitRadius=20, drivingBone {boneName=foot_r}, sphereLocalOffset=<0 0 0>}
    ]
   }
   n94 = Anim Dynamics - Bone: dyn_skirt_bk_l_01!(ComponentPose=@n93, Alpha=1, GravityOverride:Vector="0,0,-980", ExternalForce="0,0,0") -> (Pose) {
    angularDampingOverride=1
    angularSpringConstant=50
    bAngularSpring=true
    bOverrideAngularDamping=true
    bUseGravityOverride=true
    bUsePlanarLimit=false
    bUseSphericalLimits=true
    gravityScale=0.3
    simulationSpace=BoneRelative
    alphaBoolBlend {blendOption=Linear}
    boundBone {boneName=dyn_skirt_bk_l_01}
    componentAppliedLinearAccClamp=<2500 2500 1500>
    componentLinearAccScale=<0.4 0.4 1>
    componentLinearVelScale=<0.55 0.55 0.6>
    physicsBodyDefinitions [{
      collisionType=CustomSphere
      sphereCollisionRadius=2
      boundBone {boneName=dyn_skirt_bk_l_01}
      boxExtents=<20 20 20>
      constraintSetup {angularLimitsMax=<0 30 10>, angularLimitsMin=<0 -50 -10>, angularTarget=<0 0 0>}
      localJointOffset=<9 0 0>
     }]
    relativeSpaceBone {boneName=root}
    sphericalLimits [
     {limitRadius=10, drivingBone {boneName=thigh_l}, sphereLocalOffset=<-10 0 0>}
     {limitRadius=10, drivingBone {boneName=thigh_l}, sphereLocalOffset=<-30 0 0>}
    ]
   }
   n95 = Anim Dynamics - Bone: dyn_skirt_bk_l_02!(ComponentPose=@n94, Alpha=1, GravityOverride:Vector="0,0,-980", ExternalForce="0,0,0") -> (Pose) {
    angularDampingOverride=1
    angularSpringConstant=50
    bAngularSpring=true
    bOverrideAngularDamping=true
    bUseGravityOverride=true
    bUsePlanarLimit=false
    bUseSphericalLimits=true
    gravityScale=0.3
    simulationSpace=BoneRelative
    alphaBoolBlend {blendOption=Linear}
    boundBone {boneName=dyn_skirt_bk_l_02}
    componentAppliedLinearAccClamp=<2500 2500 1500>
    componentLinearAccScale=<0.4 0.4 1>
    componentLinearVelScale=<0.55 0.55 0.6>
    physicsBodyDefinitions [{
      collisionType=CustomSphere
      sphereCollisionRadius=2
      boundBone {boneName=dyn_skirt_bk_l_02}
      boxExtents=<20 20 20>
      constraintSetup {angularLimitsMax=<0 30 10>, angularLimitsMin=<0 -50 -10>, angularTarget=<0 0 0>}
      localJointOffset=<9 0 0>
     }]
    relativeSpaceBone {boneName=root}
    sphericalLimits [
     {limitRadius=10, drivingBone {boneName=thigh_l}, sphereLocalOffset=<-10 0 0>}
     {limitRadius=10, drivingBone {boneName=thigh_l}, sphereLocalOffset=<-30 0 0>}
     {limitRadius=12, drivingBone {boneName=calf_l}, sphereLocalOffset=<-5 0 0>}
    ]
   }
   n96 = Anim Dynamics - Bone: dyn_skirt_bk_l_03!(ComponentPose=@n95, Alpha=1, GravityOverride:Vector="0,0,-980", ExternalForce="0,0,0") -> (Pose) {
    angularDampingOverride=1
    angularSpringConstant=50
    bAngularSpring=true
    bOverrideAngularDamping=true
    bUseGravityOverride=true
    bUsePlanarLimit=false
    bUseSphericalLimits=true
    gravityScale=0.3
    simulationSpace=BoneRelative
    alphaBoolBlend {blendOption=Linear}
    boundBone {boneName=dyn_skirt_bk_l_03}
    componentAppliedLinearAccClamp=<2500 2500 1500>
    componentLinearAccScale=<0.4 0.4 1>
    componentLinearVelScale=<0.55 0.55 0.6>
    physicsBodyDefinitions [{
      collisionType=CustomSphere
      sphereCollisionRadius=2
      boundBone {boneName=dyn_skirt_bk_l_03}
      boxExtents=<20 20 20>
      constraintSetup {angularLimitsMax=<0 30 10>, angularLimitsMin=<0 -50 -10>, angularTarget=<0 0 0>}
      localJointOffset=<9 0 0>
     }]
    relativeSpaceBone {boneName=root}
    sphericalLimits [
     {limitRadius=10, drivingBone {boneName=thigh_l}, sphereLocalOffset=<-10 0 0>}
     {limitRadius=10, drivingBone {boneName=thigh_l}, sphereLocalOffset=<-30 0 0>}
     {limitRadius=12, drivingBone {boneName=calf_l}, sphereLocalOffset=<-5 0 0>}
     {limitRadius=15, drivingBone {boneName=calf_l}, sphereLocalOffset=<-25 0 0>}
     {limitRadius=10, drivingBone {boneName=foot_l}, sphereLocalOffset=<0 0 0>}
    ]
   }
   Reroute Node(InputPin=@n96)
   Get melee() -> (melee:double)
   Get not crouching() -> (not crouching:double)
  ]}
 EventGraph {nodes [
   n1 = Event BlueprintUpdateAnimation()
   > n2 = Sequence!(execute) -> (then_0, then_1, then_2:Exec, then_3:Exec)
   n3 = Get bIsCrouching()
   n4 = Get bIsMoving2D()
   > Set crouch_moving(crouch_moving:double="To Float (Boolean)"(InBool=AND Boolean(A=@n3, B=@n4))) -> (Output_Get:double)
   > Set crouching(crouching:double="To Float (Boolean)"(InBool=@n3)) -> (Output_Get:double)
   > Set not crouching(not crouching:double="To Float (Boolean)"(InBool=NOT Boolean(A=@n3))) -> (Output_Get:double)
   > Set moving2D(moving2D:double="To Float (Boolean)"(InBool=@n4)) -> (Output_Get:double)
   n5 = NOT Boolean(A=Get bIsMoving2D())
   n6 = Get bIsCrouching()
   > Set Crouch Still(Crouch Still:double="To Float (Boolean)"(InBool=AND Boolean(A=@n6, B=@n5))) -> (Output_Get:double)
   Is Valid(exec=@n2.then_1, InputObject=Get PlayerPawn() -> (PlayerPawn:Fort Player Pawn Athena Object Reference))
   > n7 = Branch(Condition=Get bIsInFrontEnd())
   Set WPO Globals Function(
    execute=@n7.else
    Material=Get Mat Instance Dynamic()
    Overlay Material=Get Overlay Mat Instance Dynamic()
    Skeletal Mesh=Get SkelMesh()
    "Local Noise Transform  Location"=Get Local Noise Transform Location() -> (Local Noise Transform Location:Vector)
    Local Noise Offset=Get Local Noise Offset()
   )
   > Set WPO Regional Function!(
    execute
    Material:Material Instance Dynamic Object Reference=Get Mat Instance Dynamic()
    Overlay Material:Material Instance Dynamic Object Reference=Get Overlay Mat Instance Dynamic()
    Skeletal Mesh:Skeletal Mesh Component Object Reference=Get SkelMesh()
    Velocity Transform:Name=foot_r
    Velocity Interp Speed:double=Get Velocity Interp Speed()
    Collision Transform:Name=calf_r
    Collision Transform Axis:EAxis Enum
    Invert Collision Transform Axis:bool=false
    Delta Time X:double=@n1.DeltaTimeX
    Previous Velocity Transform Location:"Vector (by ref)"=Get R1 Velocity Transform Location()
    Previous Velocity Transform Velocity:"Vector (by ref)"=Get R1 Velocity Transform Velocity()
    Region:int=1
   ) -> (then)
   > Set WPO Regional Function(
    Material=Get Mat Instance Dynamic()
    Overlay Material=Get Overlay Mat Instance Dynamic()
    Skeletal Mesh=Get SkelMesh()
    Velocity Transform=foot_l
    Velocity Interp Speed=Get Velocity Interp Speed()
    Collision Transform=calf_l
    Invert Collision Transform Axis=true
    Delta Time X=@n1.DeltaTimeX
    Previous Velocity Transform Location=Get R2 Velocity Transform Location()
    Previous Velocity Transform Velocity=Get R2 Velocity Transform Velocity()
    Region=2
   )
   > n8 = Branch(Condition=Get bIsSkydiveFloating())
   > n9 = Set WPO Pho Wind Intensity(WPO Pho Wind Intensity=1)
   SetScalarParameterValue(
    execute=[@n10.then, @n9.then]
    self=[Get Mat Instance Dynamic(), Get Overlay Mat Instance Dynamic()]
    ParameterName=ClothSim_WindStrength
    Value=Get WPO Pho Wind Intensity()
   )
   n10 = Set WPO Pho Wind Intensity(execute=@n8.else, WPO Pho Wind Intensity=0)
   n11 = SelectFloat(A=1, B=0, bPickA=OR Boolean(A=Get bIsJumping(), B=Get bIsFalling()))
   Set jumping(execute=@n2.then_2, jumping:double=@n11) -> (Output_Get:double)
   > Set melee(melee:double=SelectFloat(A=0.5, B=1, bPickA=Get bIsPlayingMeleeAnim())) -> (Output_Get:double)
   n12 = Get CurrentWeapon() -> (CurrentWeapon:Fort Weapon Object Reference)
   n13 = "Equal Exactly (String)"(A=GetDisplayName(Object=@n12), B=B_ChainWhip0)
   Set r_skirt_lock(execute=@n2.then_3, r_skirt_lock:bool=@n13) -> (Output_Get:bool)
   > Set rban_off(rban_off:bool=NOT Boolean(A=Get bRigidBodyEnabled())) -> (Output_Get:bool)
   > Set FE_Off(FE_Off:bool=NOT Boolean(A=Get bIsInFrontEnd())) -> (Output_Get:bool)
   Event BlueprintInitializeAnimation()
   > Set SkelMesh(SkelMesh=GetOwnerSkeletalMeshComponent())
   n14 = TryGetPawnOwner()
   > Is Valid(InputObject=@n14)
   > n15 = Cast To FortPlayerPawnAthena(Object=@n14)
   > Set PlayerPawn(PlayerPawn:Fort Player Pawn Athena Object Reference=@n15) -> (Output_Get:Fort Player Pawn Athena Object Reference)
   Event OnCharacterCustomizationLoaded()
   > Sequence!(execute) -> (then_0, then_1)
   > n16 = Set SkelMesh(SkelMesh=GetPartSkeletalMeshComponent())
   > n17 = Get Dynamic Material Instance(SkeletalMeshComponent=@n16)
   > Set Mat Instance Dynamic(Mat Instance Dynamic=@n17)
   > n18 = GetOrCreateDynamicOverlayMaterial(MeshComponent=@n16, MaterialName=None)
   > Set Overlay Mat Instance Dynamic(Overlay Mat Instance Dynamic=@n18)
   FInterpTo(Current=Get jumping(), Target=@n11, DeltaTime=0.5, InterpSpeed=0.6)
   n19 = GetAnimInstance(self=Get WeaponMesh!(self:Fort Weapon Object Reference=@n12) -> (WeaponMesh:Skeletal Mesh Component Budgeted Object Reference))
   AND Boolean(
    A=@n13
    B=OR Boolean!(
     A=Montage_IsPlaying(self=@n19, Montage=/SunRoseChainWhipGameplay/Items/ChainWhip/Animations/Weapon/Hades_Weapon_Attack_01_M)
     B=Montage_IsPlaying(self=@n19, Montage=/SunRoseChainWhipGameplay/Items/ChainWhip/Animations/Weapon/Hades_Weapon_Attack_02_M)
     C:bool=Montage_IsPlaying(self=@n19, Montage=/SunRoseChainWhipGameplay/Items/ChainWhip/Animations/Weapon/Hades_Weapon_Attack_03_M)
    ) -> (ReturnValue)
   )
  ]}
]
variables [
 Crouch Still {type=number}
 crouch_moving {type=number}
 crouching {type=number}
 FE_Off {type=boolean}
 jumping {type=number}
 Local Noise Offset {type=Vector}
 Local Noise Transform Location {type=Vector}
 Mat Instance Dynamic {type=MaterialInstanceDynamic}
 melee {type=number}
 moving2D {type=number}
 not crouching {type=number}
 Overlay Mat Instance Dynamic {type=MaterialInstanceDynamic}
 PlayerPawn {type=FortPlayerPawnAthena}
 R1 Velocity Transform Location {type=Vector}
 R1 Velocity Transform Velocity {type=Vector}
 R2 Velocity Transform Location {type=Vector}
 R2 Velocity Transform Velocity {type=Vector}
 r_skirt_lock {type=boolean}
 rban_off {type=boolean}
 SkelMesh {type=SkeletalMeshComponent}
 Velocity Interp Speed {type=number}
 weapon {type=string, default=""}
 WPO Pho Wind Intensity {type=number}
]
