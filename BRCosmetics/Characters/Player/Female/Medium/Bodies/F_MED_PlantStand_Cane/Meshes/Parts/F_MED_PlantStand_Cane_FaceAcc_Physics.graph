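# Physics asset for the F_MED_PlantStand_Cane female character's face accessory mesh, covering two hair sidetails and a ponytail.
# Each hair piece is a four-link capsule chain from a master bone on the head, with tight angular limits at the root and looser ones down the chain.
# Capsules on head, neck, upper spine, clavicles and upper arms form body volumes, joined by free-rotation constraints, as are the hair masters to the head.
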PhysicsAsset /BRCosmetics/Characters/Player/Female/Medium/Bodies/F_MED_PlantStand_Cane/Meshes/Parts/F_MED_PlantStand_Cane_FaceAcc_Physics
bodies {
 clavicle_l [{
   shapeType=Capsule
   length=8.1258
   radius=5.994
   shapeName=clavicle_l_capsule
   center=<7.2301 -1.2328 -1.3435>
   rotation=<rot 81.1982 -0.0006 -0.0006>
  }]
 clavicle_r [{
   shapeType=Capsule
   length=8.1258
   radius=5.994
   shapeName=clavicle_l_capsule
   center=<-6.772 1.5345 1.4654>
   rotation=<rot 81.1982 -0.0006 -0.0006>
  }]
 dyn_hair_sidetail_l_01 [{shapeType=Capsule, length=2.5034, radius=3.4966, center=<4.4851 0 3>, rotation=<rot 90 0 0>}]
 dyn_hair_sidetail_l_02 [{shapeType=Capsule, length=2.0221, radius=3.4379, shapeName=dyn_hair_sidetail_l_02_capsule, center=<2.9819 0 3>, rotation=<rot 90 0 0>}]
 dyn_hair_sidetail_l_03 [
  {shapeType=Capsule, length=2.6062, radius=3.3938, shapeName=dyn_hair_sidetail_r_03_capsule, center=<4.4851 3 0>, rotation=<rot 90 0 0>}
  {shapeType=Capsule, length=0, radius=3.6555, shapeName=dyn_hair_sidetail_r_03_capsule, center=<7.453 3 -0.1917>, rotation=<rot 90 0 0>}
 ]
 dyn_hair_sidetail_l_04 [{
   shapeType=Capsule
   length=6.3584
   radius=2.5479
   shapeName=dyn_hair_sidetail_l_04_capsule
   center=<4.4851 3 0>
   rotation=<rot 90 68.3242 74.02>
  }]
 dyn_hair_sidetail_l_mstr [{shapeType=Capsule, length=0, radius=1.6971, rotation=<rot 90 0 0>}]
 dyn_hair_sidetail_r_01 [{shapeType=Capsule, length=2.4172, radius=3.5828, center=<4.4851 0 3>, rotation=<rot 90 0 0>}]
 dyn_hair_sidetail_r_02 [{shapeType=Capsule, length=4.2793, radius=3.4, shapeName=dyn_hair_sidetail_r_02_capsule, center=<2.9819 0 3>, rotation=<rot 90 0 0>}]
 dyn_hair_sidetail_r_03 [
  {shapeType=Capsule, length=2.3618, radius=3.6382, shapeName=dyn_hair_sidetail_r_03_capsule, center=<4.4851 -3 0>, rotation=<rot 90 0 0>}
  {shapeType=Capsule, length=0, radius=3.4784, shapeName=dyn_hair_sidetail_r_03_capsule, center=<7.453 -3 0.3192>, rotation=<rot 90 0 0>}
 ]
 dyn_hair_sidetail_r_04 [{
   shapeType=Capsule
   length=5.9345
   radius=2.9718
   shapeName=dyn_hair_sidetail_r_04_capsule
   center=<4.4851 -3 0>
   rotation=<rot 90 -69.7339 -64.0379>
  }]
 dyn_hair_sidetail_r_mstr [{shapeType=Capsule, length=0, radius=1.6971, rotation=<rot 90 0 0>}]
 dyn_ponytail_01 [{shapeType=Capsule, length=3, radius=3, center=<4.0645 0 0>, rotation=<rot 90 0 0>}]
 dyn_ponytail_02 [{shapeType=Capsule, length=2.1997, radius=3.8003, shapeName=dyn_ponytail_02_capsule, center=<3.6476 0 0>, rotation=<rot 90 0 0>}]
 dyn_ponytail_03 [{shapeType=Capsule, length=3.8505, radius=3, shapeName=dyn_ponytail_03_capsule, center=<4.1736 0 0>, rotation=<rot 90 0 0>}]
 dyn_ponytail_04 [{shapeType=Capsule, length=4.0281, radius=3, shapeName=dyn_ponytail_04_capsule, center=<4.1204 0 0>, rotation=<rot 90 0 0>}]
 dyn_ponytail_mstr [{shapeType=Capsule, length=0, radius=1.6971, rotation=<rot 90 0 0>}]
 head [
  {shapeType=Capsule, length=3.2573, radius=7.3817, rotation=<rot 0 0 90>}
  {shapeType=Capsule, length=1.1275, radius=6.7171, shapeName=head_capsule, center=<7.0827 -2.825 0>}
 ]
 neck_01 [{shapeType=Capsule, length=0, radius=6.2662, shapeName=neck_01_capsule, rotation=<rot 90 0 0>}]
 neck_02 [{shapeType=Capsule, length=0, radius=6.2662, shapeName=neck_02_capsule, rotation=<rot 90 0 0>}]
 spine_04 [{shapeType=Capsule, length=20.03, radius=10, shapeName=clavicle_l_capsule, center=<0.0564 3.6341 0>}]
 spine_05 [{shapeType=Capsule, length=20.8096, radius=8.8897, shapeName=clavicle_l_capsule}]
 upperarm_l [{shapeType=Capsule, length=16.186, radius=4.15, shapeName=upperarm_l_capsule, center=<12.3474 0 -0.6308>, rotation=<rot 90 0 0>}]
 upperarm_r [{shapeType=Capsule, length=16.186, radius=4.6253, shapeName=upperarm_l_capsule, center=<-11.2684 0 0.6228>, rotation=<rot 90 0 0>}]
}
constraints [
 {bone1Name=dyn_hair_sidetail_l_01, bone2Name=dyn_hair_sidetail_l_mstr, swing1LimitDegrees=5, swing2LimitDegrees=8, twistLimitDegrees=4}
 {bone1Name=dyn_hair_sidetail_r_01, bone2Name=dyn_hair_sidetail_r_mstr, swing1LimitDegrees=5, swing2LimitDegrees=8, twistLimitDegrees=4}
 {bone1Name=dyn_ponytail_01, bone2Name=dyn_ponytail_mstr, swing1LimitDegrees=5, swing2LimitDegrees=5, twistLimitDegrees=3}
 {bone1Name=dyn_hair_sidetail_l_02, bone2Name=dyn_hair_sidetail_l_01, swing1LimitDegrees=27, swing2LimitDegrees=27, twistLimitDegrees=20}
 {bone1Name=dyn_hair_sidetail_l_03, bone2Name=dyn_hair_sidetail_l_02, swing1LimitDegrees=27, swing2LimitDegrees=27, twistLimitDegrees=20}
 {bone1Name=dyn_hair_sidetail_l_04, bone2Name=dyn_hair_sidetail_l_03, swing1LimitDegrees=27, swing2LimitDegrees=27, twistLimitDegrees=20}
 {bone1Name=dyn_hair_sidetail_r_02, bone2Name=dyn_hair_sidetail_r_01, swing1LimitDegrees=27, swing2LimitDegrees=27, twistLimitDegrees=20}
 {bone1Name=dyn_hair_sidetail_r_03, bone2Name=dyn_hair_sidetail_r_02, swing1LimitDegrees=27, swing2LimitDegrees=27, twistLimitDegrees=20}
 {bone1Name=dyn_hair_sidetail_r_04, bone2Name=dyn_hair_sidetail_r_03, swing1LimitDegrees=27, swing2LimitDegrees=27, twistLimitDegrees=20}
 {bone1Name=dyn_ponytail_02, bone2Name=dyn_ponytail_01, swing1LimitDegrees=30, swing2LimitDegrees=30, twistLimitDegrees=15}
 {bone1Name=dyn_ponytail_03, bone2Name=dyn_ponytail_02, swing1LimitDegrees=30, swing2LimitDegrees=30, twistLimitDegrees=15}
 {bone1Name=dyn_ponytail_04, bone2Name=dyn_ponytail_03, swing1LimitDegrees=30, swing2LimitDegrees=30, twistLimitDegrees=15}
 {
  bone1Name=head
  bone2Name=neck_02
  swing1LimitDegrees=45
  swing1Motion=Free
  swing2LimitDegrees=45
  swing2Motion=Free
  twistLimitDegrees=45
  twistMotion=Free
 }
 {
  bone1Name=neck_02
  bone2Name=neck_01
  swing1LimitDegrees=45
  swing1Motion=Free
  swing2LimitDegrees=45
  swing2Motion=Free
  twistLimitDegrees=45
  twistMotion=Free
 }
 {
  bone1Name=neck_01
  bone2Name=spine_05
  swing1LimitDegrees=45
  swing1Motion=Free
  swing2LimitDegrees=45
  swing2Motion=Free
  twistLimitDegrees=45
  twistMotion=Free
 }
 {
  bone1Name=clavicle_l
  bone2Name=spine_05
  swing1LimitDegrees=45
  swing1Motion=Free
  swing2LimitDegrees=45
  swing2Motion=Free
  twistLimitDegrees=45
  twistMotion=Free
 }
 {
  bone1Name=upperarm_l
  bone2Name=clavicle_l
  swing1LimitDegrees=45
  swing1Motion=Free
  swing2LimitDegrees=45
  swing2Motion=Free
  twistLimitDegrees=45
  twistMotion=Free
 }
 {
  bone1Name=clavicle_r
  bone2Name=spine_05
  swing1LimitDegrees=45
  swing1Motion=Free
  swing2LimitDegrees=45
  swing2Motion=Free
  twistLimitDegrees=45
  twistMotion=Free
 }
 {
  bone1Name=upperarm_r
  bone2Name=clavicle_r
  swing1LimitDegrees=45
  swing1Motion=Free
  swing2LimitDegrees=45
  swing2Motion=Free
  twistLimitDegrees=45
  twistMotion=Free
 }
 {
  bone1Name=dyn_hair_sidetail_l_mstr
  bone2Name=head
  swing1LimitDegrees=45
  swing1Motion=Free
  swing2LimitDegrees=45
  swing2Motion=Free
  twistLimitDegrees=45
  twistMotion=Free
 }
 {
  bone1Name=dyn_hair_sidetail_r_mstr
  bone2Name=head
  swing1LimitDegrees=45
  swing1Motion=Free
  swing2LimitDegrees=45
  swing2Motion=Free
  twistLimitDegrees=45
  twistMotion=Free
 }
 {
  bone1Name=dyn_ponytail_mstr
  bone2Name=head
  swing1LimitDegrees=45
  swing1Motion=Free
  swing2LimitDegrees=45
  swing2Motion=Free
  twistLimitDegrees=45
  twistMotion=Free
 }
 {
  bone1Name=spine_05
  bone2Name=spine_04
  swing1LimitDegrees=45
  swing1Motion=Free
  swing2LimitDegrees=45
  swing2Motion=Free
  twistLimitDegrees=45
  twistMotion=Free
 }
]
