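# Post-process-style AnimBlueprint for the PlantStand Cane female cosmetic's FaceAcc part, adding secondary motion to hair.
# It copies the parent mesh pose, layering the Body_Mesh pose from clavicles and upper arms, with Body_Mesh cached at initialization.
# The chain runs ponytail AnimDynamics, a targeting-driven right sidetail rotation, then a spine_05-based RigidBody simulation.
# An unconnected RigidBody node referencing RoyalDuskBlaze physics remains in the graph.
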AnimBlueprint /BRCosmetics/Characters/Player/Female/Medium/Bodies/F_MED_PlantStand_Cane/Meshes/Parts/F_MED_PlantStand_Cane_FaceAcc_AnimBP
target_skeleton=&/F_MED_PlantStand_Cane_FaceAcc_Skeleton
graphs [
 AnimGraph {nodes [
   RigidBody!(ComponentPose, Alpha, OverrideWorldGravity:Vector="0,0,0", ExternalForce) -> (Pose) {
    bClampLinearTranslationLimitToRefPose=true
    bOverrideWorldGravity=true
    overridePhysicsAsset=/BRCosmetics/Characters/Player/Female/Medium/Bodies/F_MED_RoyalDuskBlaze/Meshes/Parts/F_MED_RoyalDuskBlaze_FaceAcc_Physics
    simulationSpace=BaseBoneSpace
    alphaBoolBlend {blendOption=Linear}
    baseBoneRef {boneName=spine_05}
    componentAppliedLinearAccClamp=<5000 5000 5000>
    componentLinearAccScale=<0.6 0.2 0.2>
    componentLinearVelScale=<0.3 0.1 0.1>
    overrideWorldGravity {z=-980}
    simSpaceSettings {
     maxAngularAcceleration=1000
     maxAngularVelocity=3
     maxLinearAcceleration=700
     maxLinearVelocity=500
     velocityScaleZ=0
     worldAlpha=1
     externalLinearDragV=<0.2 0.2>
    }
   }
   n1 = Layered blend per bone(
    BasePose=Copy Pose From Mesh() {bCopyCurves=true, bUseAttachedParent=true}
    BlendPoses_0=Copy Pose From Mesh(SourceMeshComponent=Get Body_Mesh()) {bCopyCurves=true, bCopyCustomAttributes=true, bUseAttachedParent=true, bUseMeshPose=true}
   ) {layerSetup [{branchFilters [{boneName=upperarm_l}, {boneName=upperarm_r}, {boneName=clavicle_r}, {boneName=clavicle_l}]}]}
   n2 = Anim Dynamics - Bone: dyn_ponytail_01!(ComponentPose=Local To Component(LocalPose=@n1), Alpha=1, GravityOverride:Vector="0,0,-980", ExternalForce="0,0,-980") -> (Pose) {
    angularDampingOverride=0.99
    angularSpringConstant=200
    bGravityOverrideInSimSpace=true
    bOverrideAngularDamping=true
    bUseGravityOverride=true
    gravityScale=0.3
    simulationSpace=BoneRelative
    alphaBoolBlend {blendOption=Linear}
    boundBone {boneName=dyn_ponytail_01}
    componentAppliedLinearAccClamp=<2500 2500 2500>
    componentLinearAccScale=<0.5 0.4 0.55>
    componentLinearVelScale=<0.55 0.55 0.55>
    physicsBodyDefinitions [{
      sphereCollisionRadius=0
      boundBone {boneName=dyn_ponytail_01}
      boxExtents=<30 30 30>
      constraintSetup {angularLimitsMax=<0 8 3>, angularLimitsMin=<0 -8 -3>, angularTarget=<0 0 0>}
      localJointOffset=<9 0 0>
     }]
    relativeSpaceBone {boneName=root}
   }
   n3 = "Transform (Modify) Bone - Bone: dyn_hair_sidetail_r_mstr"!(ComponentPose=@n2, bAlphaBoolEnabled:bool=Get bIsTargeting(), Translation="0,0,0", Rotation="10,0,0", Scale="1,1,1") -> (Pose) {
    alphaInputType=Bool
    rotationMode=BMM_Additive
    rotationSpace=BCS_BoneSpace
    translationMode=BMM_Additive
    translationSpace=BCS_BoneSpace
    alphaBoolBlend {blendOption=Linear}
    boneToModify {boneName=dyn_hair_sidetail_r_mstr}
    rotation {pitch=10}
   }
   n4 = RigidBody(ComponentPose=@n3, OverrideWorldGravity="0,0,-980", ExternalForce="0,0,-980") {
    bClampLinearTranslationLimitToRefPose=true
    bOverrideWorldGravity=true
    overridePhysicsAsset=&/F_MED_PlantStand_Cane_FaceAcc_Physics
    simulationSpace=BaseBoneSpace
    alphaBoolBlend {blendOption=Linear}
    baseBoneRef {boneName=spine_05}
    componentAppliedLinearAccClamp=<1500 1500 1500>
    componentLinearAccScale=<0.4 0.2 0.2>
    overrideWorldGravity {z=-980}
    simSpaceSettings {
     maxAngularAcceleration=100
     maxAngularVelocity=3
     maxLinearAcceleration=700
     maxLinearVelocity=500
     velocityScaleZ=0
     worldAlpha=1
     externalLinearDragV=<0.4 0.4 0.4>
    }
   }
   Output Pose(Result=Component To Local(ComponentPose=@n4)) {layerGroup=None}
  ]}
 EventGraph {nodes [
   Event BlueprintInitializeAnimation()
   > Set Body_Mesh(Body_Mesh=GetOwnerSkeletalMeshComponent())
   GetSkeletalMeshForPartType(self=Cast To PlayerPawn_Athena(Object=TryGetPawnOwner()))
   Event BlueprintUpdateAnimation()
  ]}
]
variables [Body_Mesh {type=SkeletalMeshComponent}]
